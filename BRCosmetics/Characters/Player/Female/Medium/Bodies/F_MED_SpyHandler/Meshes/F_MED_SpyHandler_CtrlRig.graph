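# Control Rig for the F_MED_SpyHandler cosmetic body that adds procedural corrective motion to her jacket, skirt and gun bones.
# It runs debug-only limb controls, a shared DeformRig function, then jacket, skirt and gun functions in sequence.
# Spherical pose readers on thighs and left upper arm, thigh-up anim attributes, pelvis height and sliding/targeting variables weight additive bone offsets.
# The skirt also follows right-thigh swing, split by sign into front and back bones.
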ControlRigBlueprint /BRCosmetics/Characters/Player/Female/Medium/Bodies/F_MED_SpyHandler/Meshes/F_MED_SpyHandler_CtrlRig
graphs [
 New Function_ContainedGraph {nodes [
   n1 = RigVMFunctionEntryNode()
   RigVMFunction_ControlFlowBranch = &RigVMFunction_ControlFlowBranch(ExecuteContext=@n1.ExecuteContext, Condition=@n1.debug)
   RigVMFunctionReturnNode(ExecuteContext=@RigVMFunction_ControlFlowBranch.Completed)
   RigUnit_GetTransform = &RigUnit_GetTransform(Item=(Type=Control,Name="thigh_l_ctrl"), bInitial=False)
   RigUnit_SetTransform = &RigUnit_SetTransform(
    Item=(Type=Bone,Name="thigh_l")
    Weight=1
    bInitial=False
    Value=@RigUnit_GetTransform.Transform
    ExecutePin=@RigVMFunction_ControlFlowBranch.True
   )
   > &RigUnit_SetTransform(
    Item=(Type=Bone,Name="thigh_r")
    Weight=1
    bInitial=False
    Value=&RigUnit_GetTransform(Item=(Type=Control,Name="thigh_r_ctrl"), bInitial=False).Transform
   )
   > &RigUnit_SetTransform(
    Item=(Type=Bone,Name="upperarm_l")
    Weight=1
    bInitial=False
    Value=&RigUnit_GetTransform(Item=(Type=Control,Name="upperarm_l_ctrl"), bInitial=False).Transform
   )
  ]}
 New Function_ContainedGraph {nodes [
   RigVMFunctionEntryNode()
   > RigVMFunction_Sequence = RigVMAggregateNode()
   &RigUnit_ModifyTransforms(
    ItemToModify=((Item=(Type=Bone,Name="base_jacket_0_l"),Transform=(Rotation=(X=0,Y=0,Z=-0.2588,W=0.9659),Translation=(X=0,Y=0,Z=0),Scale3D=(X=1,Y=1,Z=1))))
    WeightMinimum=0
    Weight=RigVMDispatchNode(BoneName=root, Default=0, Name=deformRig_l_thigh_up_alpha) {template="DISPATCH_RigDispatch_GetAnimAttribute(in Name,in BoneName,in Default,out Value,out Found)"}.Value
    ExecutePin=@RigVMFunction_Sequence.A
   )
   > &RigUnit_ModifyTransforms(
    ItemToModify=((Item=(Type=Bone,Name="base_jacket_0_r"),Transform=(Rotation=(X=0,Y=0,Z=-0.1994,W=0.9799),Translation=(X=0,Y=0,Z=0),Scale3D=(X=1,Y=1,Z=1))))
    WeightMinimum=0
    Weight=RigVMDispatchNode(BoneName=root, Default=0, Name=deformRig_r_thigh_up_alpha) {template="DISPATCH_RigDispatch_GetAnimAttribute(in Name,in BoneName,in Default,out Value,out Found)"}.Value
   )
   > n1 = &RigUnit_SphericalPoseReader(
    ActiveRegionScaleFactors=(PositiveWidth=0.1,NegativeWidth=1,PositiveHeight=1,NegativeHeight=1)
    ActiveRegionSize=0.1
    Debug=(bDrawDebug=false,bDraw2D=False,bDrawLocalAxes=False,DebugScale=25,DebugSegments=20,DebugThickness=0.25)
    DriverAxis=(X=-1,Y=0,Z=0)
    DriverItem=(Type=Bone,Name="thigh_l")
    FalloffRegionScaleFactors=(PositiveWidth=0.7,NegativeWidth=1,PositiveHeight=1,NegativeHeight=1)
    FalloffSize=0.2
    FlipHeightScaling=False
    FlipWidthScaling=False
    OptionalParentItem=(Type=Bone,Name="None")
    RotationOffset=(X=0,Y=180,Z=0)
   )
   > &RigUnit_ModifyTransforms(
    ItemToModify="((Item=(Type=Bone,Name=\"base_jacket_0_l\"),Transform=(Rotation=(X=0,Y=0,Z=-0.2588,W=0.9659),Translation=(X=0,Y=0,Z=0),Scale3D=(X=1,Y=1,Z=1))),(Item=(Type=Bone,Name=\"base_jacket_1_l\"),Transform=(Rotation=(X=0,Y=0,Z=-0.342,W=0.9397),Translation=(X=0,Y=0,Z=0), ...(truncated)"
    WeightMinimum=0
    Weight=@n1.OutputParam
   )
   > n2 = &RigUnit_SphericalPoseReader(
    ActiveRegionScaleFactors=(PositiveWidth=0.1,NegativeWidth=1,PositiveHeight=1,NegativeHeight=1)
    ActiveRegionSize=0.1
    Debug=(bDrawDebug=false,bDraw2D=False,bDrawLocalAxes=False,DebugScale=25,DebugSegments=20,DebugThickness=0.25)
    DriverAxis=(X=1,Y=0,Z=0)
    DriverItem=(Type=Bone,Name="thigh_r")
    FalloffRegionScaleFactors=(PositiveWidth=0.7,NegativeWidth=1,PositiveHeight=1,NegativeHeight=1)
    FalloffSize=0.2
    FlipHeightScaling=False
    FlipWidthScaling=False
    OptionalParentItem=(Type=Bone,Name="None")
    RotationOffset=(X=0,Y=0,Z=0)
   )
   > &RigUnit_ModifyTransforms(
    ItemToModify="((Item=(Type=Bone,Name=\"base_jacket_0_r\"),Transform=(Rotation=(X=0,Y=0,Z=-0.2588,W=0.9659),Translation=(X=0,Y=0,Z=0),Scale3D=(X=1,Y=1,Z=1))),(Item=(Type=Bone,Name=\"base_jacket_1_r\"),Transform=(Rotation=(X=0,Y=0,Z=-0.342,W=0.9397),Translation=(X=0,Y=0,Z=0), ...(truncated)"
    WeightMinimum=0
    Weight=@n2.OutputParam
   )
   > RigVMFunctionReturnNode()
   RigVMFunction_MathBoolToFloat = &RigVMFunction_MathBoolToFloat(Value=RigVMVariableNode() {variable=IsSlidingBaseball}.Value)
   &RigUnit_ModifyTransforms(
    ItemToModify="((Item=(Type=Bone,Name=\"base_jacket_0_r\"),Transform=(Rotation=(X=0,Y=0,Z=-0.342,W=0.9397),Translation=(X=0,Y=0,Z=0),Scale3D=(X=1,Y=1,Z=1))),(Item=(Type=Bone,Name=\"base_jacket_1_r\"),Transform=(Rotation=(X=-0.0594,Y=-0.1632,Z=-0.3368,W=0.9254),Translation=(X ...(truncated)"
    WeightMinimum=0
    Weight=@RigVMFunction_MathBoolToFloat.Result
    ExecutePin=@RigVMFunction_Sequence.C
   )
   Remap = &RigVMFunction_MathDoubleRemap(
    SourceMaximum=45
    SourceMinimum=89
    TargetMaximum=1
    TargetMinimum=0
    Value=&RigUnit_GetRelativeTransformForItem(Child=(Type=Bone,Name="pelvis"), Parent=(Type=Bone,Name="root"), bChildInitial=False, bParentInitial=False).RelativeTransform.Translation.Z
   )
   RigVMVariableNode(Value=@Remap.Result, ExecuteContext=@RigVMFunction_Sequence.B) {variable=lowToGround}
   n3 = RigVMVariableNode() {variable=lowToGround}
   DISPATCH_RigVMDispatch_If = RigVMDispatchNode(False=1, True=0, Condition=RigVMVariableNode() {variable=IsSliding}.Value) {template="DISPATCH_RigVMDispatch_If(in Condition,in True,in False,out Result)"}
   Multiply = &RigVMFunction_MathDoubleMul(A=@n3.Value, B=@DISPATCH_RigVMDispatch_If.Result)
   > &RigUnit_ModifyTransforms(
    ItemToModify="((Item=(Type=Bone,Name=\"base_jacket_0_l\"),Transform=(Rotation=(X=0,Y=0,Z=-0.5373,W=0.8434),Translation=(X=0,Y=0,Z=0),Scale3D=(X=1,Y=1,Z=1))),(Item=(Type=Bone,Name=\"dyn_jacket_0_r\"),Transform=(Rotation=(X=0,Y=0,Z=-0.1736,W=0.9848),Translation=(X=0,Y=0,Z=0), ...(truncated)"
    WeightMinimum=0
    Weight=@Multiply.Result
   )
   > n4 = &RigUnit_SphericalPoseReader(
    ActiveRegionScaleFactors=(PositiveWidth=1,NegativeWidth=1,PositiveHeight=1,NegativeHeight=1)
    ActiveRegionSize=0.1
    Debug=(bDrawDebug=false,bDraw2D=False,bDrawLocalAxes=False,DebugScale=25,DebugSegments=20,DebugThickness=0.25)
    DriverAxis=(X=1,Y=0,Z=0)
    DriverItem=(Type=Bone,Name="upperarm_l")
    FalloffRegionScaleFactors=(PositiveWidth=1,NegativeWidth=1,PositiveHeight=1,NegativeHeight=1)
    FalloffSize=0.2
    FlipHeightScaling=False
    FlipWidthScaling=False
    OptionalParentItem=(Type=Bone,Name="None")
    RotationOffset=(X=50,Y=0,Z=-40)
   )
   > &RigUnit_ModifyTransforms(
    ItemToModify="((Item=(Type=Bone,Name=\"base_jacket_0_l\"),Transform=(Rotation=(X=0,Y=0,Z=0.2798,W=0.96),Translation=(X=0,Y=0,Z=0),Scale3D=(X=1,Y=1,Z=1))),(Item=(Type=Bone,Name=\"dyn_jacket_0_r\"),Transform=(Rotation=(X=0,Y=0,Z=0.0872,W=0.9962),Translation=(X=0,Y=0,Z=0),Scal ...(truncated)"
    WeightMinimum=0
    Weight=&RigVMFunction_MathDoubleMul(A=@n4.OutputParam, B=@n3.Value).Result
   )
  ]}
 New Function_ContainedGraph {nodes [
   RigVMFunctionEntryNode()
   > RigVMFunction_Sequence = RigVMAggregateNode()
   n1 = RigVMDispatchNode(BoneName=root, Default=0, Name=deformRig_l_thigh_up_alpha) {template="DISPATCH_RigDispatch_GetAnimAttribute(in Name,in BoneName,in Default,out Value,out Found)"}
   &RigUnit_ModifyTransforms(
    ItemToModify="((Item=(Type=Bone,Name=\"dyn_skirt_front\"),Transform=(Rotation=(X=0,Y=0,Z=0,W=1),Translation=(X=30,Y=0,Z=20),Scale3D=(X=1,Y=1,Z=1))),(Item=(Type=Bone,Name=\"dyn_skirt_back\"),Transform=(Rotation=(X=0,Y=0,Z=0,W=1),Translation=(X=0,Y=0,Z=0),Scale3D=(X=1,Y=1,Z=1 ...(truncated)"
    WeightMinimum=0
    Weight=@n1.Value
    ExecutePin=@RigVMFunction_Sequence.A
   )
   n2 = RigVMDispatchNode(BoneName=root, Default=0, Name=deformRig_r_thigh_up_alpha) {template="DISPATCH_RigDispatch_GetAnimAttribute(in Name,in BoneName,in Default,out Value,out Found)"}
   > &RigUnit_ModifyTransforms(
    ItemToModify=((Item=(Type=Bone,Name="dyn_skirt_front"),Transform=(Rotation=(X=0,Y=0,Z=0,W=1),Translation=(X=87.9497,Y=18.5665,Z=-41.0552),Scale3D=(X=1,Y=1,Z=1))))
    WeightMinimum=0
    Weight=@n2.Value
   )
   Add = &RigVMFunction_MathDoubleAdd(A=@n1.Value, B=@n2.Value)
   Remap = &RigVMFunction_MathDoubleRemap(SourceMaximum=2, SourceMinimum=1, TargetMaximum=1, TargetMinimum=0, Value=@Add.Result)
   > &RigUnit_ModifyTransforms(
    ItemToModify="((Item=(Type=Bone,Name=\"dyn_skirt_front\"),Transform=(Rotation=(X=0,Y=0,Z=0,W=1),Translation=(X=-80,Y=0,Z=0),Scale3D=(X=1,Y=1,Z=1))),(Item=(Type=Bone,Name=\"base_skirt_back\"),Transform=(Rotation=(X=0,Y=0,Z=0,W=1),Translation=(X=0,Y=-50,Z=0),Scale3D=(X=1,Y=1, ...(truncated)"
    WeightMinimum=0
    Weight=@Remap.Result
   )
   > n3 = &RigUnit_SphericalPoseReader(
    ActiveRegionScaleFactors=(PositiveWidth=1,NegativeWidth=1.3,PositiveHeight=1,NegativeHeight=1)
    ActiveRegionSize=0.11
    Debug=(bDrawDebug=false,bDraw2D=False,bDrawLocalAxes=False,DebugScale=25,DebugSegments=20,DebugThickness=0.25)
    DriverAxis=(X=-1,Y=0,Z=0)
    DriverItem=(Type=Bone,Name="thigh_l")
    FalloffRegionScaleFactors=(PositiveWidth=1,NegativeWidth=3.5,PositiveHeight=1,NegativeHeight=1)
    FalloffSize=0.215
    FlipHeightScaling=False
    FlipWidthScaling=False
    OptionalParentItem=(Type=Bone,Name="None")
    RotationOffset=(X=-90,Y=0,Z=-25)
   )
   > &RigUnit_ModifyTransforms(
    ItemToModify="((Item=(Type=Bone,Name=\"dyn_skirt_front\"),Transform=(Rotation=(X=0,Y=0,Z=0,W=1),Translation=(X=0,Y=15,Z=0),Scale3D=(X=1,Y=1,Z=1))),(Item=(Type=Bone,Name=\"base_skirt_back\"),Transform=(Rotation=(X=0,Y=0,Z=0,W=1),Translation=(X=-70,Y=-70,Z=-20),Scale3D=(X=1,Y ...(truncated)"
    WeightMinimum=0
    Weight=@n3.OutputParam
   )
   > n4 = &RigUnit_SphericalPoseReader(
    ActiveRegionScaleFactors=(PositiveWidth=1,NegativeWidth=1.5,PositiveHeight=1,NegativeHeight=0.35)
    ActiveRegionSize=0.1
    Debug=(bDrawDebug=false,bDraw2D=False,bDrawLocalAxes=False,DebugScale=25,DebugSegments=20,DebugThickness=0.25)
    DriverAxis=(X=-1,Y=0,Z=0)
    DriverItem=(Type=Bone,Name="thigh_l")
    FalloffRegionScaleFactors=(PositiveWidth=1,NegativeWidth=3,PositiveHeight=1,NegativeHeight=0.45)
    FalloffSize=0.2
    FlipHeightScaling=False
    FlipWidthScaling=False
    OptionalParentItem=(Type=Bone,Name="None")
    RotationOffset=(X=-60,Y=0,Z=-30)
   )
   > &RigUnit_ModifyTransforms(
    ItemToModify="((Item=(Type=Bone,Name=\"dyn_skirt_front\"),Transform=(Rotation=(X=0,Y=0,Z=0,W=1),Translation=(X=0,Y=0,Z=-35),Scale3D=(X=1,Y=1,Z=1))),(Item=(Type=Bone,Name=\"base_skirt_back\"),Transform=(Rotation=(X=0,Y=0,Z=0,W=1),Translation=(X=-55,Y=0,Z=-80),Scale3D=(X=1,Y= ...(truncated)"
    WeightMinimum=0
    Weight=@n4.OutputParam
   )
   n5 = &RigUnit_SphericalPoseReader(
    ActiveRegionScaleFactors=(PositiveWidth=1,NegativeWidth=1,PositiveHeight=1,NegativeHeight=1)
    ActiveRegionSize=0.1
    Debug=(bDrawDebug=false,bDraw2D=False,bDrawLocalAxes=False,DebugScale=25,DebugSegments=20,DebugThickness=0.25)
    DriverAxis=(X=1,Y=0,Z=0)
    DriverItem=(Type=Bone,Name="thigh_r")
    FalloffRegionScaleFactors=(PositiveWidth=4,NegativeWidth=1,PositiveHeight=4,NegativeHeight=1)
    FalloffSize=0.2
    FlipHeightScaling=False
    FlipWidthScaling=False
    OptionalParentItem=(Type=Bone,Name="None")
    RotationOffset=(X=-142.0862,Y=20,Z=40)
    ExecutePin=@RigVMFunction_Sequence.C
   )
   Multiply = &RigVMFunction_MathDoubleMul(A=@n5.OutputParam, B=RigVMVariableNode() {variable=isTargeting}.Value)
   > &RigUnit_ModifyTransforms(
    ItemToModify=((Item=(Type=Bone,Name="base_skirt_front"),Transform=(Rotation=(X=0,Y=0,Z=0,W=1),Translation=(X=7.2138,Y=12.0772,Z=-16.7728),Scale3D=(X=1,Y=1,Z=1))))
    WeightMinimum=0
    Weight=@Multiply.Result
   )
   > n6 = &RigUnit_SphericalPoseReader(
    ActiveRegionScaleFactors=(PositiveWidth=1,NegativeWidth=1,PositiveHeight=1,NegativeHeight=1)
    ActiveRegionSize=0.05
    Debug=(bDrawDebug=false,bDraw2D=False,bDrawLocalAxes=False,DebugScale=25,DebugSegments=20,DebugThickness=0.25)
    DriverAxis=(X=-1,Y=0,Z=0)
    DriverItem=(Type=Bone,Name="thigh_l")
    FalloffRegionScaleFactors=(PositiveWidth=1,NegativeWidth=1,PositiveHeight=1,NegativeHeight=1)
    FalloffSize=0.15
    FlipHeightScaling=False
    FlipWidthScaling=False
    OptionalParentItem=(Type=Bone,Name="None")
    RotationOffset=(X=0,Y=72,Z=-2)
   )
   > &RigUnit_ModifyTransforms(
    ItemToModify="((Item=(Type=Bone,Name=\"base_skirt_back\"),Transform=(Rotation=(X=0,Y=0,Z=0,W=1),Translation=(X=0,Y=-10,Z=-20),Scale3D=(X=1,Y=1,Z=1))),(Item=(Type=Bone,Name=\"base_back_skirt\"),Transform=(Rotation=(X=0.0076,Y=0.0868,Z=-0.0868,W=0.9924),Translation=(X=0,Y=0,Z ...(truncated)"
    WeightMinimum=0
    Weight=&RigVMFunction_MathDoubleMul(A=@n6.OutputParam, B=RigVMVariableNode() {variable=isTargeting}.Value).Result
   )
   &RigUnit_ParentConstraint(
    AdvancedSettings=(InterpolationType=Average,RotationOrderForFilter=XZY)
    Child=(Type=Bone,Name="thigh_front_skirt_cons")
    Filter=(TranslationFilter=(bX=True,bY=True,bZ=True),RotationFilter=(bX=true,bY=True,bZ=True),ScaleFilter=(bX=True,bY=True,bZ=True))
    Parents=((Item=(Type=Bone,Name="thigh_r"),Weight=1))
    Weight=1
    ExecutePin=@RigVMFunction_Sequence.B
   )
   RigUnit_GetTransform = &RigUnit_GetTransform(Item=(Type=Bone,Name="thigh_front_skirt_cons"), Space=LocalSpace, bInitial=false)
   RigVMFunction_MathQuaternionSwingTwist = &RigVMFunction_MathQuaternionSwingTwist(TwistAxis=(X=1,Y=0,Z=0), Input=@RigUnit_GetTransform.Transform.Rotation)
   RigVMFunction_MathQuaternionToEuler = &RigVMFunction_MathQuaternionToEuler(RotationOrder=ZYX, Value=@RigVMFunction_MathQuaternionSwingTwist.Swing)
   LessEqual = /Script/RigVM.RigVMFunction_MathDoubleLessEqual(B=0, A=@RigVMFunction_MathQuaternionToEuler.Result.Z) {is=struct RigVMUnitNode, template="LessEqual::Execute(in A,in B,out Result)"}
   > &RigUnit_ModifyTransforms(
    ItemToModify=((Item=(Type=Bone,Name="thigh_front_skirt_modify"),Transform=(Rotation=(X=0,Y=0,Z=0,W=1),Translation=(X=0,Y=0,Z=0),Scale3D=(X=1,Y=1,Z=1))))
    Weight=1
    WeightMinimum=0
    ItemToModify.0.Transform.Rotation=&RigVMFunction_MathQuaternionFromEuler(
     Euler=(X=0,Y=0,Z=0)
     Euler.Z=RigVMDispatchNode(False=0, Condition=@LessEqual.Result, True=@RigVMFunction_MathQuaternionToEuler.Result.Z) {template="DISPATCH_RigVMDispatch_If(in Condition,in True,in False,out Result)"}.Result
     Euler.Y=@RigVMFunction_MathQuaternionToEuler.Result.Y
     Euler.X=@RigVMFunction_MathQuaternionToEuler.Result.X
    ).Result
   )
   > &RigUnit_ModifyTransforms(
    ItemToModify=((Item=(Type=Bone,Name="thigh_front_skirt_modify"),Transform=(Rotation=(X=0,Y=0,Z=0.025,W=0.9997),Translation=(X=0.0889,Y=-1.7749,Z=0),Scale3D=(X=1,Y=1,Z=1))))
    WeightMinimum=0
    Weight=RigVMDispatchNode(BoneName=root, Default=0, Name=deformRig_r_thigh_up_alpha) {template="DISPATCH_RigDispatch_GetAnimAttribute(in Name,in BoneName,in Default,out Value,out Found)"}.Value
   )
   > n7 = &RigUnit_SphericalPoseReader(
    ActiveRegionScaleFactors=(PositiveWidth=3,NegativeWidth=1,PositiveHeight=1.2,NegativeHeight=1)
    ActiveRegionSize=0.1
    Debug=(bDrawDebug=false,bDraw2D=False,bDrawLocalAxes=False,DebugScale=25,DebugSegments=20,DebugThickness=0.25)
    DriverAxis=(X=1,Y=0,Z=0)
    DriverItem=(Type=Bone,Name="thigh_r")
    FalloffRegionScaleFactors=(PositiveWidth=1,NegativeWidth=1,PositiveHeight=0.3,NegativeHeight=0.3)
    FalloffSize=0.58
    FlipHeightScaling=true
    FlipWidthScaling=true
    OptionalParentItem=(Type=Bone,Name="None")
    RotationOffset=(X=-135,Y=0,Z=-15)
   )
   > &RigUnit_ModifyTransforms(
    ItemToModify="((Item=(Type=Bone,Name=\"thigh_front_skirt_modify\"),Transform=(Rotation=(X=0.4527,Y=-0.028,Z=-0.0766,W=0.8879),Translation=(X=-0.1922,Y=-0.7412,Z=-3.5851),Scale3D=(X=1,Y=1,Z=1))),(Item=(Type=Bone,Name=\"dyn_skirt_front\"),Transform=(Rotation=(X=0,Y=0,Z=0,W=1) ...(truncated)"
    WeightMinimum=0
    Weight=@n7.OutputParam
   )
   GreaterEqual = /Script/RigVM.RigVMFunction_MathDoubleGreaterEqual(B=0, A=@RigVMFunction_MathQuaternionToEuler.Result.Z) {is=struct RigVMUnitNode, template="GreaterEqual::Execute(in A,in B,out Result)"}
   RigVMFunction_MathQuaternionFromEuler = &RigVMFunction_MathQuaternionFromEuler(
    Euler=(X=0,Y=0,Z=0)
    Euler.Z=RigVMDispatchNode(False=0, Condition=@GreaterEqual.Result, True=@RigVMFunction_MathQuaternionToEuler.Result.Z) {template="DISPATCH_RigVMDispatch_If(in Condition,in True,in False,out Result)"}.Result
    Euler.Y=@RigVMFunction_MathQuaternionToEuler.Result.Y
    Euler.X=@RigVMFunction_MathQuaternionToEuler.Result.X
   )
   > &RigUnit_ModifyTransforms(
    ItemToModify=((Item=(Type=Bone,Name="thigh_back_skirt_modify"),Transform=(Rotation=(X=0,Y=0,Z=0,W=1),Translation=(X=0,Y=0,Z=0),Scale3D=(X=1,Y=1,Z=1))))
    Weight=1
    WeightMinimum=0
    ItemToModify.0.Transform.Rotation=@RigVMFunction_MathQuaternionFromEuler.Result
   )
   > &RigUnit_ParentConstraint(
    AdvancedSettings=(InterpolationType=Average,RotationOrderForFilter=XZY)
    Child=(Type=Bone,Name="base_back_skirt")
    Parents=((Item=(Type=Bone,Name="thigh_back_skirt_modify"),Weight=1))
    Weight=1
   )
   > &RigUnit_ParentConstraint(
    AdvancedSettings=(InterpolationType=Average,RotationOrderForFilter=XZY)
    Child=(Type=Bone,Name="base_skirt_0")
    Parents=((Item=(Type=Bone,Name="thigh_front_skirt_modify"),Weight=1))
    Weight=1
   )
   > n8 = &RigUnit_SphericalPoseReader(
    ActiveRegionScaleFactors=(PositiveWidth=1,NegativeWidth=1,PositiveHeight=1,NegativeHeight=1)
    ActiveRegionSize=0.1
    Debug=(bDrawDebug=false,bDraw2D=False,bDrawLocalAxes=False,DebugScale=25,DebugSegments=20,DebugThickness=0.25)
    DriverAxis=(X=1,Y=0,Z=0)
    DriverItem=(Type=Bone,Name="thigh_r")
    FalloffRegionScaleFactors=(PositiveWidth=1,NegativeWidth=1,PositiveHeight=1.5,NegativeHeight=1)
    FalloffSize=0.35
    FlipHeightScaling=False
    FlipWidthScaling=False
    OptionalParentItem=(Type=Bone,Name="None")
    RotationOffset=(X=0,Y=-20,Z=0)
   )
   > &RigUnit_ModifyTransforms(
    ItemToModify=((Item=(Type=Bone,Name="base_back_skirt"),Transform=(Rotation=(X=0,Y=0,Z=0.1045,W=0.9945),Translation=(X=0,Y=-5,Z=0),Scale3D=(X=1,Y=1,Z=1))))
    WeightMinimum=0
    Weight=@n8.OutputParam
   )
   > n9 = &RigUnit_SphericalPoseReader(
    ActiveRegionScaleFactors=(PositiveWidth=1,NegativeWidth=1,PositiveHeight=1,NegativeHeight=1)
    ActiveRegionSize=0.1
    Debug=(bDrawDebug=false,bDraw2D=False,bDrawLocalAxes=False,DebugScale=25,DebugSegments=20,DebugThickness=0.25)
    DriverAxis=(X=1,Y=0,Z=0)
    DriverItem=(Type=Bone,Name="thigh_r")
    FalloffRegionScaleFactors=(PositiveWidth=1,NegativeWidth=1,PositiveHeight=1,NegativeHeight=1)
    FalloffSize=0.2
    FlipHeightScaling=False
    FlipWidthScaling=False
    OptionalParentItem=(Type=Bone,Name="None")
    RotationOffset=(X=90,Y=0,Z=-20)
   )
   > &RigUnit_ModifyTransforms(
    ItemToModify="((Item=(Type=Bone,Name=\"base_skirt_back\"),Transform=(Rotation=(X=0,Y=0,Z=0,W=1),Translation=(X=-18.1554,Y=-14.0707,Z=0),Scale3D=(X=1,Y=1,Z=1))),(Item=(Type=Bone,Name=\"base_back_skirt\"),Transform=(Rotation=(X=0,Y=-0.0086,Z=0,W=1),Translation=(X=-0.0164,Y=-0 ...(truncated)"
    WeightMinimum=0
    Weight=@n9.OutputParam
   )
   > RigVMFunctionReturnNode()
  ]}
 New Function_ContainedGraph {nodes [
   RigVMFunctionEntryNode()
   > n1 = &RigUnit_SphericalPoseReader(
    ActiveRegionScaleFactors=(PositiveWidth=1,NegativeWidth=1,PositiveHeight=0.5,NegativeHeight=1)
    ActiveRegionSize=0.1
    Debug=(bDrawDebug=false,bDraw2D=false,bDrawLocalAxes=False,DebugScale=25,DebugSegments=20,DebugThickness=0.25)
    DriverAxis=(X=1,Y=0,Z=0)
    DriverItem=(Type=Bone,Name="upperarm_l")
    FalloffRegionScaleFactors=(PositiveWidth=2,NegativeWidth=2,PositiveHeight=1,NegativeHeight=2)
    FalloffSize=0.2
    FlipHeightScaling=False
    FlipWidthScaling=False
    OptionalParentItem=(Type=Bone,Name="None")
    RotationOffset=(X=-136.6181,Y=31.6852,Z=101.1581)
   )
   AlphaInterp = &RigVMFunction_AlphaInterp(InterpSpeedDecreasing=8, InterpSpeedIncreasing=8, Scale=1, bClampResult=False, Value=@n1.OutputParam)
   > &RigUnit_ModifyTransforms(
    ItemToModify=((Item=(Type=Bone,Name="dyn_gun"),Transform=(Rotation=(X=0,Y=0,Z=-0.4226,W=0.9063),Translation=(X=1.2534,Y=0,Z=0),Scale3D=(X=1,Y=1,Z=1))))
    WeightMinimum=0
    Weight=@AlphaInterp.Result
   )
   > RigVMFunctionReturnNode()
  ]}
 RigVMFunctionLibrary {nodes [
   controls = RigVMCollapseNode(ExecuteContext=(), debug=False)
   gun = RigVMCollapseNode(ExecuteContext=())
   jacket = RigVMCollapseNode(ExecuteContext=())
   skirt = RigVMCollapseNode(ExecuteContext=())
  ]}
 RigVMFunction_Sequence_1_ContainedGraph {nodes [
   n1 = RigVMFunctionEntryNode()
   RigVMFunction_Sequence = &RigVMFunction_Sequence(ExecuteContext=@n1.ExecuteContext)
   n2 = &RigVMFunction_Sequence(ExecuteContext=@RigVMFunction_Sequence.B)
   RigVMFunctionReturnNode(A=@RigVMFunction_Sequence.A, B=@n2.A, C=@n2.B)
  ]}
 RigVMFunction_Sequence_1_ContainedGraph {nodes [
   n1 = RigVMFunctionEntryNode()
   RigVMFunction_Sequence = &RigVMFunction_Sequence(ExecuteContext=@n1.ExecuteContext)
   n2 = &RigVMFunction_Sequence(ExecuteContext=@RigVMFunction_Sequence.B)
   RigVMFunctionReturnNode(A=@RigVMFunction_Sequence.A, B=@n2.A, C=@n2.B)
  ]}
 RigVMFunction_Sequence_1_ContainedGraph {nodes [
   n1 = RigVMFunctionEntryNode()
   RigVMFunction_Sequence = &RigVMFunction_Sequence(ExecuteContext=@n1.ExecuteContext)
   n2 = &RigVMFunction_Sequence(ExecuteContext=@RigVMFunction_Sequence.B)
   RigVMFunctionReturnNode(A=@RigVMFunction_Sequence.A, B=@n2.A, C=@n2.B)
  ]}
 RigVMModel {nodes [
   RigUnit_BeginExecution = &RigUnit_BeginExecution()
   RigVMFunction_Sequence = RigVMAggregateNode(ExecuteContext=@RigUnit_BeginExecution.ExecutePin)
   controls = RigVMFunctionReferenceNode(debug=false, ExecuteContext=@RigVMFunction_Sequence.A)
   FN_DeformRig_FNC = RigVMFunctionReferenceNode(
    DeformRig_ Curve="(deformRig_l_upperarm_up_alpha_72_#1=false,deformRig_r_upperarm_up_alpha_73_#2=False,deformRig_l_upperarm_fr_alpha_74_#3=False,deformRig_r_upperarm_fr_alpha_75_CB51B8 ...(truncated)"
    DeformRig_Bool="(upperarm_1_#4=True,elbow_3_#5=True,wrist_7_#6=True,groin_8_#7=True,glute_17_#8=True,knee_9_40C408324CA261 ...(truncated)"
    DeformRig_Debug=(upperarm_debug_18_#4=False,wrist_debug_20_#6=False,lat_debug_26_#9=false,groinGlute_debug_27_#7=false)
    DeformRig_Value="(upperarm_up_18_#10=(Rotation=(X=0.0,Y=-0.0,Z=0.0,W=1.0),Translation=(X=0.0,Y=0.0,Z=8.0),Scale3D=(X=0.70,Y=1.0,Z=1.0)),upperarm_fwd_fr_19_#11=(Rotation=(X=0.0,Y=-0.0,Z=0.0,W=1.0),Translation=(X=0.0, ...(truncated)"
    ExecuteContext=@RigVMFunction_Sequence.B
   )
   jacket = RigVMFunctionReferenceNode(ExecuteContext=@RigVMFunction_Sequence.C)
   > skirt = RigVMFunctionReferenceNode()
   > gun = RigVMFunctionReferenceNode()
  ]}
]
hierarchy [
 root {loc=<0 0 0>, rot=<0 0 0 1>}
 attach {parent=root, loc=<0 0 0>, rot=<0 0 0 1>}
 pelvis {parent=root, loc=<0 -0.5301 98.0571>, rot=<0.0127 0.707 -0.0127 -0.707>}
 spine_01 {parent=pelvis, loc=<6.1208 -0.9328 0>, rot=<0 0 0.034 -0.9994>}
 spine_02 {parent=spine_01, loc=<6.5996 -0.2095 0>, rot=<0 0 0.0238 -0.9997>}
 spine_03 {parent=spine_02, loc=<6.4914 -0.1237 0>, rot=<0 0 0.0149 -0.9999>}
 spine_04 {parent=spine_03, loc=<6.9598 0.0794 0>, rot=<0 0 -0.0132 -0.9999>}
 spine_05 {parent=spine_04, loc=<6.9987 0 0>, rot=<0 0 -0.0015 -1>}
 clavicle_l {parent=spine_05, loc=<5.4831 -0.1053 -3.2946>, rot=<0.0465 -0.7289 -0.009 -0.6829>}
 upperarm_l {parent=clavicle_l, loc=<9.8658 -1.2136 -0.4105>, rot=<0.0164 -0.3245 -0.0269 -0.9454>}
 lowerarm_l {parent=upperarm_l, loc=<26.465 -0.8345 0.259>, rot=<0.0036 0.0161 -0.2298 -0.9731>}
 hand_l {parent=lowerarm_l, loc=<22.7355 -0.3203 -0.6105>, rot=<-0.6784 -0.0223 0.0735 -0.7307>}
 index_metacarpal_l {parent=hand_l, loc=<1.6535 0.0183 -2.96>, rot=<0.0158 -0.0124 -0.0173 -0.9996>}
 index_01_l {parent=index_metacarpal_l, loc=<8.3008 0.0357 -0.0524>, rot=<-0.0044 0.0485 0.0924 -0.9945>}
 index_02_l {parent=index_01_l, loc=<3.3587 -0.3004 0.0452>, rot=<-0.002 0.0027 0.3225 -0.9466>}
 index_03_l {parent=index_02_l, loc=<2.5578 -0.123 -0.0252>, rot=<-0.002 0.0067 0.2558 -0.9667>}
 middle_metacarpal_l {parent=hand_l, loc=<2.1856 -0.1725 -0.5984>, rot=<0.006 0.0168 -0.0329 -0.9993>}
 middle_01_l {parent=middle_metacarpal_l, loc=<7.3275 -0.2447 0.0129>, rot=<-0.0208 0.0042 0.0858 -0.9961>}
 middle_02_l {parent=middle_01_l, loc=<4.0123 0.1401 0.1175>, rot=<-0.0218 0.0258 0.3512 -0.9357>}
 middle_03_l {parent=middle_02_l, loc=<2.6077 0.0534 0.1373>, rot=<0.0082 0.0188 0.1708 -0.9851>}
 pinky_metacarpal_l {parent=hand_l, loc=<1.7616 -0.9753 3.3814>, rot=<-0.0397 0.0814 -0.069 -0.9935>}
 pinky_01_l {parent=pinky_metacarpal_l, loc=<7.1709 -0.4831 -0.1094>, rot=<-0.0654 -0.0035 0.2242 -0.9723>}
 pinky_02_l {parent=pinky_01_l, loc=<2.8274 0.0893 -0.1043>, rot=<0.0037 0.0011 0.2854 -0.9584>}
 pinky_03_l {parent=pinky_02_l, loc=<2.2046 0.0307 -0.0078>, rot=<-0.0002 0.0098 0.1445 -0.9895>}
 ring_metacarpal_l {parent=hand_l, loc=<2.2074 -0.3625 1.3638>, rot=<-0.0284 0.0722 -0.0458 -0.9959>}
 ring_01_l {parent=ring_metacarpal_l, loc=<7.0295 -0.6197 -0.1446>, rot=<-0.0489 -0.0252 0.1892 -0.9804>}
 ring_02_l {parent=ring_01_l, loc=<3.6174 0.4647 -0.1027>, rot=<0 0 0.2806 -0.9598>}
 ring_03_l {parent=ring_02_l, loc=<2.4152 0.114 0.0199>, rot=<0 0 0.1189 -0.9929>}
 thumb_01_l {parent=hand_l, loc=<1.3118 -0.3191 -2.459>, rot=<0.6546 -0.4042 0.0664 -0.6354>}
 thumb_02_l {parent=thumb_01_l, loc=<5.9336 -0.0448 0.0401>, rot=<-0.0037 0.0213 0.298 -0.9543>}
 thumb_03_l {parent=thumb_02_l, loc=<2.2223 0.3682 -0.0907>, rot=<0.0118 -0.013 0.0066 -0.9998>}
 weapon_l {parent=hand_l, loc=<-0.1974 -3.4111 0.3811>, rot=<0 0 0 1>}
 deform_wrist_dn_base_l {parent=hand_l, loc=<0 -1.5326 0>, rot=<0 0 0 -1>}
 deform_wrist_dn_l {parent=deform_wrist_dn_base_l, loc=<0 0 0>, rot=<0 0 0 1>}
 deform_wrist_up_base_l {parent=hand_l, loc=<0 2.3009 0>, rot=<0 0 0 -1>}
 deform_wrist_up_l {parent=deform_wrist_up_base_l, loc=<0 0 0>, rot=<0 0 0 1>}
 lowerarm_twist_01_l {parent=lowerarm_l, loc=<18.316 0 0>, rot=<0 0 0 1>}
 lowerarm_twist_02_l {parent=lowerarm_l, loc=<10.6707 0 0>, rot=<0 0 0 1>}
 deform_elbow_in_base_l {parent=lowerarm_l, loc=<0.1834 3.6062 -0.2572>, rot=<0 0 0 -1>}
 deform_elbow_in_l {parent=deform_elbow_in_base_l, loc=<0 0 0>, rot=<0 0 0 1>}
 deform_elbow_out_base_l {parent=lowerarm_l, loc=<0 -4 0>, rot=<0 0 0 1>}
 deform_elbow_out_l {parent=deform_elbow_out_base_l, loc=<0 0 0>, rot=<0 0 0 1>}
 base_Bracelete {parent=lowerarm_l, loc=<0 0 0>, rot=<0.8184 -0.0069 -0.0037 -0.5746>}
 dyn_backBracelete_00_l {parent=base_Bracelete, loc=<0 0 0>, rot=<-0.0008 -0.0012 0.0031 -1>}
 dyn_backBracelete_01_l {parent=dyn_backBracelete_00_l, loc=<20.0909 0 0>, rot=<0 0 0 1>}
 dyn_CP_backBracelete_l {parent=dyn_backBracelete_01_l, loc=<-0.1219 0.0927 2.8134>, rot=<-0.4851 -0.5257 0.4604 -0.5257>}
 dyn_frontBracelete_00_l {parent=base_Bracelete, loc=<0 0 0>, rot=<0 0 0 1>}
 dyn_frontBracelete_01_l {parent=dyn_frontBracelete_00_l, loc=<21.9818 0 0>, rot=<0 0 0 1>}
 dyn_CP_frontBracelete_l {parent=dyn_frontBracelete_01_l, loc=<-0.1621 0.1755 2.8592>, rot=<-0.4983 -0.5259 0.45 -0.5221>}
 upperarm_twist_01_l {parent=upperarm_l, loc=<7.2135 -0.1134 0>, rot=<0 0 0 1>}
 deform_upperarm_bk_base_l {parent=upperarm_twist_01_l, loc=<-6.4053 -4.3746 0.314>, rot=<0 0 0 -1>}
 deform_upperarm_bk_l {parent=deform_upperarm_bk_base_l, loc=<0 0 0>, rot=<0 0 0 1>}
 deform_upperarm_fr_base_l {parent=upperarm_twist_01_l, loc=<-5.6312 5.092 0.3354>, rot=<0 0 0 -1>}
 deform_upperarm_fr_l {parent=deform_upperarm_fr_base_l, loc=<0 0 0>, rot=<0 0 0 1>}
 deform_upperarm_up_base_l {parent=upperarm_twist_01_l, loc=<-8.891 0.4825 3.9517>, rot=<0 0 0 -1>}
 deform_upperarm_up_l {parent=deform_upperarm_up_base_l, loc=<0 0 0>, rot=<0 0 0 1>}
 upperarm_twist_02_l {parent=upperarm_l, loc=<19.7998 -0.1134 0>, rot=<0 0 0 1>}
 clavicle_r {parent=spine_05, loc=<5.4831 -0.1053 3.2946>, rot=<0.7289 0.0465 -0.6829 0.009>}
 upperarm_r {parent=clavicle_r, loc=<-9.8658 1.2136 0.4104>, rot=<0.0164 -0.3245 -0.0269 -0.9454>}
 lowerarm_r {parent=upperarm_r, loc=<-26.4651 0.8345 -0.259>, rot=<0.0036 0.0161 -0.2298 -0.9731>}
 hand_r {parent=lowerarm_r, loc=<-22.7356 0.3203 0.6106>, rot=<-0.6784 -0.0223 0.0735 -0.7307>}
 index_metacarpal_r {parent=hand_r, loc=<-1.6534 -0.0181 2.9601>, rot=<0.0158 -0.0124 -0.0173 -0.9996>}
 index_01_r {parent=index_metacarpal_r, loc=<-8.3008 -0.0357 0.0524>, rot=<-0.0044 0.0485 0.0924 -0.9945>}
 index_02_r {parent=index_01_r, loc=<-3.3589 0.3003 -0.0453>, rot=<-0.002 0.0027 0.3225 -0.9466>}
 index_03_r {parent=index_02_r, loc=<-2.5578 0.123 0.0252>, rot=<-0.002 0.0067 0.2558 -0.9667>}
 middle_metacarpal_r {parent=hand_r, loc=<-2.1856 0.1726 0.5984>, rot=<0.006 0.0168 -0.0329 -0.9993>}
 middle_01_r {parent=middle_metacarpal_r, loc=<-7.3273 0.2447 -0.0129>, rot=<-0.0208 0.0042 0.0858 -0.9961>}
 middle_02_r {parent=middle_01_r, loc=<-4.0124 -0.14 -0.1175>, rot=<-0.0218 0.0258 0.3512 -0.9357>}
 middle_03_r {parent=middle_02_r, loc=<-2.6077 -0.0533 -0.1373>, rot=<0.0082 0.0188 0.1708 -0.9851>}
 pinky_metacarpal_r {parent=hand_r, loc=<-1.7617 0.9753 -3.3814>, rot=<-0.0397 0.0814 -0.069 -0.9935>}
 pinky_01_r {parent=pinky_metacarpal_r, loc=<-7.1707 0.4832 0.1094>, rot=<-0.0654 -0.0035 0.2242 -0.9723>}
 pinky_02_r {parent=pinky_01_r, loc=<-2.8274 -0.0892 0.1043>, rot=<0.0037 0.0011 0.2854 -0.9584>}
 pinky_03_r {parent=pinky_02_r, loc=<-2.2045 -0.031 0.0078>, rot=<-0.0002 0.0098 0.1445 -0.9895>}
 ring_metacarpal_r {parent=hand_r, loc=<-2.2073 0.3627 -1.3638>, rot=<-0.0284 0.0722 -0.0458 -0.9959>}
 ring_01_r {parent=ring_metacarpal_r, loc=<-7.0297 0.6197 0.1446>, rot=<-0.0489 -0.0252 0.1892 -0.9804>}
 ring_02_r {parent=ring_01_r, loc=<-3.6171 -0.4648 0.1027>, rot=<0 0 0.2806 -0.9598>}
 ring_03_r {parent=ring_02_r, loc=<-2.4153 -0.1141 -0.0199>, rot=<0 0 0.1189 -0.9929>}
 thumb_01_r {parent=hand_r, loc=<-1.3119 0.3192 2.459>, rot=<0.6546 -0.4042 0.0664 -0.6354>}
 thumb_02_r {parent=thumb_01_r, loc=<-5.9335 0.0447 -0.0401>, rot=<-0.0037 0.0213 0.298 -0.9543>}
 thumb_03_r {parent=thumb_02_r, loc=<-2.2223 -0.3682 0.0907>, rot=<0.0118 -0.013 0.0066 -0.9998>}
 weapon_r {parent=hand_r, loc=<0.1975 3.4112 -0.3811>, rot=<0 0 0 -1>}
 deform_wrist_dn_base_r {parent=hand_r, loc=<0 1.5329 0>, rot=<0 0 0 -1>}
 deform_wrist_dn_r {parent=deform_wrist_dn_base_r, loc=<0 0 0>, rot=<0 0 0 1>}
 deform_wrist_up_base_r {parent=hand_r, loc=<-0.0001 -2.3006 -0.0001>, rot=<0 0 0 -1>}
 deform_wrist_up_r {parent=deform_wrist_up_base_r, loc=<0 0 0>, rot=<0 0 0 1>}
 lowerarm_twist_01_r {parent=lowerarm_r, loc=<-18.3161 0 0>, rot=<0 0 0 1>}
 lowerarm_twist_02_r {parent=lowerarm_r, loc=<-10.6707 0 0>, rot=<0 0 0 1>}
 deform_elbow_in_base_r {parent=lowerarm_r, loc=<-0.1834 -3.6062 0.2572>, rot=<0 0 0 -1>}
 deform_elbow_in_r {parent=deform_elbow_in_base_r, loc=<0 0 0>, rot=<0 0 0 1>}
 deform_elbow_out_base_r {parent=lowerarm_r, loc=<0 4 0>, rot=<0 0 0 1>}
 deform_elbow_out_r {parent=deform_elbow_out_base_r, loc=<0 0 0>, rot=<0 0 0 1>}
 upperarm_twist_01_r {parent=upperarm_r, loc=<-7.2135 0.1134 0>, rot=<0 0 0 1>}
 deform_upperarm_bk_base_r {parent=upperarm_twist_01_r, loc=<6.4052 4.3746 -0.3139>, rot=<0 0 0 -1>}
 deform_upperarm_bk_r {parent=deform_upperarm_bk_base_r, loc=<0 0 0>, rot=<0 0 0 1>}
 deform_upperarm_fr_base_r {parent=upperarm_twist_01_r, loc=<5.6311 -5.092 -0.3353>, rot=<0 0 0 -1>}
 deform_upperarm_fr_r {parent=deform_upperarm_fr_base_r, loc=<0 0 0>, rot=<0 0 0 1>}
 deform_upperarm_up_base_r {parent=upperarm_twist_01_r, loc=<8.8909 -0.4825 -3.9516>, rot=<0 0 0 -1>}
 deform_upperarm_up_r {parent=deform_upperarm_up_base_r, loc=<0 0 0>, rot=<0 0 0 1>}
 upperarm_twist_02_r {parent=upperarm_r, loc=<-19.7998 0.1134 0>, rot=<0 0 0 1>}
 neck_01 {parent=spine_05, loc=<11.3138 0.1529 0>, rot=<0 0 -0.2003 -0.9797>}
 neck_02 {parent=neck_01, loc=<4.5627 0 0>, rot=<0 0 0.0504 -0.9987>}
 head {parent=neck_02, loc=<5.3914 0 0>, rot=<0 0 0.115 -0.9934>}
 base_FrontHair_L {parent=head, loc=<9.4854 5.3945 -5.0204>, rot=<-0.0034 -0.9811 -0.1051 -0.1626>}
 dyn_FrontHair_L_01 {parent=base_FrontHair_L, loc=<0 0 0>, rot=<0 0 0 1>}
 dyn_FrontHair_L_02 {parent=dyn_FrontHair_L_01, loc=<4.1438 0 0>, rot=<0 0 0.0166 -0.9999>}
 dyn_FrontHair_L_03 {parent=dyn_FrontHair_L_02, loc=<5.2406 -0.0717 -0.3498>, rot=<0.0046 -0.0858 0.004 -0.9963>}
 dyn_FrontHair_L_04 {parent=dyn_FrontHair_L_03, loc=<4.2934 -0.0763 -0.1349>, rot=<0.0012 -0.1232 -0.0207 -0.9922>}
 base_FrontBangB_L {parent=dyn_FrontHair_L_02, loc=<2.2865 3.1957 -0.868>, rot=<0.5558 0.0977 -0.2727 -0.7792>}
 dyn_FrontBangB_L_01 {parent=base_FrontBangB_L, loc=<0 0 0>, rot=<0 0 0 1>}
 base_FrontBangA_L {parent=dyn_FrontHair_L_01, loc=<2.9208 3.583 -1.0282>, rot=<0.5772 0.0756 -0.0985 -0.8071>}
 dyn_FrontBangA_L_01 {parent=base_FrontBangA_L, loc=<0 0 0>, rot=<0 0 0 1>}
 base_SideHairA_L {parent=head, loc=<2.7517 0.2095 -7.5988>, rot=<0.0169 -0.9866 -0.1018 -0.126>}
 dyn_SideHairA_L_01 {parent=base_SideHairA_L, loc=<0 0 0>, rot=<0 0 0 1>}
 dyn_SideHairA_L_02 {parent=dyn_SideHairA_L_01, loc=<3.8037 -0.0726 -0.1902>, rot=<0.0044 -0.1074 0.0042 -0.9942>}
 dyn_SideHairA_L_03 {parent=dyn_SideHairA_L_02, loc=<4.2934 -0.0763 -0.1349>, rot=<0.0009 -0.1102 -0.0208 -0.9937>}
 base_frontBang_R {parent=head, loc=<9.1251 5.4077 4.9999>, rot=<0.1802 0.5192 0.8355 -0.0026>}
 dyn_frontBang_R_01 {parent=base_frontBang_R, loc=<0 0 0>, rot=<0 0 0 1>}
 dyn_frontBang_R_02 {parent=dyn_frontBang_R_01, loc=<2.2934 -0.0299 -0.0493>, rot=<0.2299 -0.2063 -0.5584 -0.7699>}
 base_BackHair_L {parent=head, loc=<3.9134 -4.405 -5.1923>, rot=<0.0558 -0.6823 0.7259 -0.0668>}
 dyn_BackHair_L_01 {parent=base_BackHair_L, loc=<0 0 0>, rot=<0 0.0165 0 -0.9999>}
 dyn_BackHair_L_02 {parent=dyn_BackHair_L_01, loc=<5.4288 0.027 -0.2843>, rot=<-0.0168 0.0101 -0.0065 -0.9998>}
 dyn_BackHair_L_03 {parent=dyn_BackHair_L_02, loc=<4.0858 0 0.1456>, rot=<0 0.0471 0 -0.9989>}
 dyn_BackHair_L_04 {parent=dyn_BackHair_L_03, loc=<3.5624 0 -0.067>, rot=<0 -0.086 0 -0.9963>}
 base_BackHair_M {parent=head, loc=<3.7377 -6.1979 0.8742>, rot=<0.0727 -0.6688 0.7357 -0.0781>}
 dyn_BackHair_M_01 {parent=base_BackHair_M, loc=<0 0 0>, rot=<0 0 0 1>}
 dyn_BackHair_M_02 {parent=dyn_BackHair_M_01, loc=<4.9094 0.0001 -0.025>, rot=<-0.0027 0.1612 -0.0116 -0.9868>}
 dyn_BackHair_M_03 {parent=dyn_BackHair_M_02, loc=<4.0858 0 0.1456>, rot=<0 -0.0383 0 -0.9993>}
 dyn_BackHair_M_04 {parent=dyn_BackHair_M_03, loc=<3.5624 0 -0.067>, rot=<0 -0.086 0 -0.9963>}
 dyn_BackHair_M_05 {parent=dyn_BackHair_M_04, loc=<3.9784 0 0.0162>, rot=<0 -0.0418 0 -0.9991>}
 base_BackHair_R {parent=head, loc=<3.708 -4.6007 6.2267>, rot=<0.0651 -0.6624 0.7446 -0.0499>}
 dyn_BackHair_R_01 {parent=base_BackHair_R, loc=<0 0 0>, rot=<0 0 0 1>}
 dyn_BackHair_R_02 {parent=dyn_BackHair_R_01, loc=<4.8187 -0.0396 0.0251>, rot=<-0.0011 0.1852 -0.003 -0.9827>}
 dyn_BackHair_R_03 {parent=dyn_BackHair_R_02, loc=<4.0858 0 0.1456>, rot=<0 -0.0551 0 -0.9985>}
 dyn_BackHair_R_04 {parent=dyn_BackHair_R_03, loc=<3.5624 0 -0.067>, rot=<0 -0.086 0 -0.9963>}
 base_BackHairPiece_R {parent=dyn_BackHair_R_03, loc=<0.7359 -2.3397 -2.2908>, rot=<-0.1423 -0.0715 0.1684 -0.9728>}
 dyn_BackHairPiece_R_01 {parent=base_BackHairPiece_R, loc=<0 0 0>, rot=<0 0 0 1>}
 base_FrontHairB_R {parent=dyn_BackHair_R_01, loc=<5.6534 -0.5863 -2.2038>, rot=<0.8745 0.0815 -0.0663 0.4735>}
 dyn_FrontHairB_R_01 {parent=base_FrontHairB_R, loc=<0 0 0>, rot=<0 0 0 1>}
 base_FrontHairA_R {parent=dyn_BackHair_R_01, loc=<3.239 0.2495 -3.2474>, rot=<0.9121 0.1965 -0.0582 0.355>}
 dyn_FrontHairA_R_01 {parent=base_FrontHairA_R, loc=<0 0 0>, rot=<0 0 0 1>}
 base_Earring_L {parent=head, loc=<-0.2911 0.4282 -5.9496>, rot=<0.476 -0.5242 -0.5014 -0.4972>}
 base_Stud_L {parent=base_Earring_L, loc=<0.1666 -0.126 0.1851>, rot=<0 0 -0.7071 -0.7071>}
 base_Hoop_L {parent=base_Earring_L, loc=<0 0 0>, rot=<0 0 -0.7071 -0.7071>}
 dyn_Hoop_L {parent=base_Hoop_L, loc=<0 0 0>, rot=<0 0 0 1>}
 base_Earring_R {parent=head, loc=<-0.2911 0.4282 5.95>, rot=<0.476 -0.5242 -0.5014 -0.4972>}
 base_Stud_R {parent=base_Earring_R, loc=<-0.1795 -0.1079 0.1841>, rot=<0 0 -0.7071 -0.7071>}
 base_Hoop_R {parent=base_Earring_R, loc=<0 0 0>, rot=<0 0 -0.7071 -0.7071>}
 dyn_Hoop_R {parent=base_Hoop_R, loc=<0 0 0>, rot=<0 0 0 1>}
 base_Cuff_R {parent=head, loc=<1.2866 -0.9617 6.8913>, rot=<-0.2118 0.7067 0.4131 -0.5339>}
 base_BackHair_CorrectiveJoint {parent=head, loc=<-1.934 -5.631 0.721>, rot=<-0.4688 -0.5282 0.5716 0.4177>}
 base_BackHair_CorrectiveJoint_01 {parent=base_BackHair_CorrectiveJoint, loc=<0 0 0>, rot=<0 0 0 1>}
 FACIAL_C_FacialRoot {parent=head, loc=<1.8137 0.4624 0.0001>, rot=<0.4768 -0.5222 -0.4768 -0.5222>}
 FACIAL_R_ForeheadOut {parent=FACIAL_C_FacialRoot, loc=<-4.8564 -1.9081 5.7018>, rot=<-0.1033 0.2571 0.0743 -0.958>}
 FACIAL_R_ForeheadMid {parent=FACIAL_C_FacialRoot, loc=<-3.5581 -2.3279 6.6977>, rot=<-0.0546 0.1759 0.0177 -0.9827>}
 FACIAL_R_ForeheadIn {parent=FACIAL_C_FacialRoot, loc=<-1.602 -1.7101 7.4502>, rot=<-0.0454 0.0695 -0.0053 -0.9965>}
 base_necklace {parent=spine_05, loc=<9.5475 1.7121 0.0518>, rot=<0 0 0.9225 0.3859>}
 dyn_necklace {parent=base_necklace, loc=<0 0 0>, rot=<0 0 0 1>}
 deform_pec_aim_base_l {parent=spine_04, loc=<3.2456 9.3309 -1.2237>, rot=<0.0603 -0.4033 0.3318 -0.8507>}
 deform_pec_aim_l {parent=deform_pec_aim_base_l, loc=<0 0 0>, rot=<0 0 0 1>}
 deform_pec_base_l {parent=deform_pec_aim_l, loc=<9.3665 0 0>, rot=<0 0 0 1>}
 deform_pec_l {parent=deform_pec_base_l, loc=<0 0 0>, rot=<0 0 0 1>}
 deform_pec_aim_base_r {parent=spine_04, loc=<3.2452 9.3309 1.2237>, rot=<-0.4033 -0.0603 0.8507 0.3318>}
 deform_pec_aim_r {parent=deform_pec_aim_base_r, loc=<0 0 0>, rot=<0 0 0 1>}
 deform_pec_base_r {parent=deform_pec_aim_r, loc=<-9.3667 -0.0003 -0.0004>, rot=<0 0 0 1>}
 deform_pec_r {parent=deform_pec_base_r, loc=<0 0 0>, rot=<0 0 0 1>}
 base_gun {parent=spine_03, loc=<4.0255 4.8407 -13.7568>, rot=<-0.2633 -0.9556 0.0063 0.132>}
 dyn_gun {parent=base_gun, loc=<0 0 0>, rot=<0 0 0 1>}
 base_jacket_0_l {parent=spine_02, loc=<0.9998 7.0999 -7.8684>, rot=<-0.1786 0.5564 0.8078 0.0773>}
 dyn_jacket_0_l {parent=base_jacket_0_l, loc=<0 0 0>, rot=<0 0 0 1>}
 dyn_jacket_0_end_l {parent=dyn_jacket_0_l, loc=<8 0 0>, rot=<0 0 0 1>}
 base_jacket_1_l {parent=spine_02, loc=<1.7085 0.5369 -9.468>, rot=<0.2222 -0.744 -0.6198 -0.114>}
 dyn_jacket_1_l {parent=base_jacket_1_l, loc=<0 0 0>, rot=<0 0 0 1>}
 dyn_jacket_1_end_l {parent=dyn_jacket_1_l, loc=<6.8692 0 0>, rot=<0 0 0 1>}
 base_jacket_2_l {parent=spine_02, loc=<2.2479 -4.1699 -6.6079>, rot=<0.2742 -0.917 -0.2774 -0.084>}
 dyn_jacket_2_l {parent=base_jacket_2_l, loc=<0 0 0>, rot=<0 0 0 1>}
 dyn_jacket_2_end_l {parent=dyn_jacket_2_l, loc=<6.3354 0 0>, rot=<0 0 0 1>}
 base_jacket_back_l {parent=spine_02, loc=<2.4289 -5.8325 0>, rot=<0.2756 -0.9613 0 0>}
 dyn_jacket_back_l {parent=base_jacket_back_l, loc=<0 0 0>, rot=<0 0 0 1>}
 dyn_jacket_back_end_l {parent=dyn_jacket_back_l, loc=<6.4462 0 0>, rot=<0 0 0 1>}
 base_jacket_1_r {parent=spine_02, loc=<1.7088 0.537 9.468>, rot=<0.2222 -0.744 0.6198 0.114>}
 dyn_jacket_1_r {parent=base_jacket_1_r, loc=<0 0 0>, rot=<0 0 0 1>}
 dyn_jacket_1_end_r {parent=dyn_jacket_1_r, loc=<6.8692 0 0>, rot=<0 0 0 1>}
 base_jacket_2_r {parent=spine_02, loc=<2.2476 -4.1699 6.6079>, rot=<0.2742 -0.917 0.2774 0.084>}
 dyn_jacket_2_r {parent=base_jacket_2_r, loc=<0 0 0>, rot=<0 0 0 1>}
 dyn_jacket_2_end_r {parent=dyn_jacket_2_r, loc=<6.3354 0 0>, rot=<0 0 0 1>}
 base_jacket_0_r {parent=spine_02, loc=<0.9996 7.0999 7.8684>, rot=<0.1786 -0.5564 0.8078 0.0773>}
 dyn_jacket_0_r {parent=base_jacket_0_r, loc=<0 0 0>, rot=<0 0 0 1>}
 dyn_jacket_0_end_r {parent=dyn_jacket_0_r, loc=<8 0 0>, rot=<0 0 0 1>}
 base_skirt_front {parent=pelvis, loc=<-20.4446 9.3253 -0.0031>, rot=<0 0 0 1>}
 dyn_skirt_front {parent=base_skirt_front, loc=<0 0 0>, rot=<0 0 0 1>}
 base_skirt_back {parent=pelvis, loc=<-16.8993 -12.4262 1.4>, rot=<-0.1071 -0.9943 0 0>}
 dyn_skirt_back {parent=base_skirt_back, loc=<0 0 0>, rot=<0 0 0 1>}
 thigh_l {parent=pelvis, loc=<-1.0855 -0.2109 -8.7053>, rot=<0.0009 0.0355 0.0155 -0.9993>}
 calf_l {parent=thigh_l, loc=<-42.5284 -0.0019 -0.7297>, rot=<0.0006 -0.0161 -0.0391 -0.9991>}
 calf_twist_01_l {parent=calf_l, loc=<-35.8483 -1.9197 -0.5794>, rot=<0.0237 0.0095 -0.0035 -0.9997>}
 calf_twist_02_l {parent=calf_l, loc=<-19.1546 -1.7943 -0.2638>, rot=<0.0237 0.0095 -0.0035 -0.9997>}
 foot_l {parent=calf_l, loc=<-46.7402 -2.5014 -0.592>, rot=<0.0427 -0.0238 0.0353 -0.9982>}
 ball_l {parent=foot_l, loc=<-3.5658 10.8066 -0.1544>, rot=<0.0001 -0.0001 -0.7063 -0.7079>}
 deform_knee_in_base_l {parent=calf_l, loc=<0.8031 -5.0231 0.2456>, rot=<0 0 0 -1>}
 deform_knee_in_l {parent=deform_knee_in_base_l, loc=<0 0 0>, rot=<0 0 0 1>}
 deform_knee_out_base_l {parent=calf_l, loc=<0 4 0>, rot=<0 0 0 1>}
 deform_knee_out_l {parent=deform_knee_out_base_l, loc=<0 0 0>, rot=<0 0 0 1>}
 base_knife {parent=calf_l, loc=<-12.588 -2.893 -8.16>, rot=<0.019 -0.0045 0.9972 -0.0719>}
 dyn_knife {parent=base_knife, loc=<0 0 0>, rot=<0 0 0 1>}
 thigh_twist_01_l {parent=thigh_l, loc=<-27.952 0 0>, rot=<0.0013 0 0 -1>}
 deform_groin_base_l {parent=thigh_l, loc=<0.4801 5.6776 -0.2486>, rot=<0 0 0 -1>}
 deform_groin_l {parent=deform_groin_base_l, loc=<0 0 0>, rot=<0 0 0 1>}
 thigh_r {parent=pelvis, loc=<-1.0853 -0.2109 8.7053>, rot=<0.0355 -0.0009 0.9993 0.0155>}
 calf_r {parent=thigh_r, loc=<42.5286 0.0019 0.7297>, rot=<0.0006 -0.0161 -0.0391 -0.9991>}
 calf_twist_01_r {parent=calf_r, loc=<35.8485 1.9197 0.5794>, rot=<0.0237 0.0095 -0.0035 -0.9997>}
 calf_twist_02_r {parent=calf_r, loc=<19.1546 1.7943 0.2638>, rot=<0.0237 0.0095 -0.0035 -0.9997>}
 foot_r {parent=calf_r, loc=<46.7403 2.5014 0.592>, rot=<0.0427 -0.0238 0.0353 -0.9982>}
 ball_r {parent=foot_r, loc=<3.5658 -10.8066 0.1544>, rot=<0.0001 -0.0001 -0.7063 -0.7079>}
 deform_knee_in_base_r {parent=calf_r, loc=<-0.8031 5.0231 -0.2457>, rot=<0 0 0 -1>}
 deform_knee_in_r {parent=deform_knee_in_base_r, loc=<0 0 0>, rot=<0 0 0 1>}
 deform_knee_out_base_r {parent=calf_r, loc=<0 -4 0>, rot=<0 0 0 1>}
 deform_knee_out_r {parent=deform_knee_out_base_r, loc=<0 0 0>, rot=<0 0 0 1>}
 thigh_twist_01_r {parent=thigh_r, loc=<27.952 0 0>, rot=<0.0013 0 0 -1>}
 deform_groin_base_r {parent=thigh_r, loc=<-0.4799 -5.6776 0.2485>, rot=<0 0 0 -1>}
 deform_groin_r {parent=deform_groin_base_r, loc=<0 0 0>, rot=<0 0 0 1>}
 deform_glute_base_l {parent=pelvis, loc=<-7.3466 -9.0213 -9.1245>, rot=<0.0172 0.031 -0.4477 -0.8935>}
 deform_glute_l {parent=deform_glute_base_l, loc=<0 0 0>, rot=<0 0 0 1>}
 deform_glute_base_r {parent=pelvis, loc=<-7.3464 -9.0213 9.1245>, rot=<0.031 -0.0172 0.8935 -0.4477>}
 deform_glute_r {parent=deform_glute_base_r, loc=<0 0 0>, rot=<0 0 0 1>}
 base_skirt_0 {parent=pelvis, loc=<-2.1348 5.3182 14.1159>, rot=<0 0 0.9975 0.0705>}
 dyn_skirt_0 {parent=base_skirt_0, loc=<0 0 0>, rot=<0 0 0 1>}
 dyn_skirt_1 {parent=dyn_skirt_0, loc=<7.6353 0 0>, rot=<0 0 0 1>}
 dyn_skirt_2 {parent=dyn_skirt_1, loc=<7.6353 0 0>, rot=<0 0 0 1>}
 dyn_skirt_3 {parent=dyn_skirt_2, loc=<7.6353 0 0>, rot=<0 0 0 1>}
 base_back_skirt {parent=pelvis, loc=<-3.1518 -8.1814 11.0833>, rot=<0.0018 -0.7074 0.7065 0.0237>}
 dyn_back_skirt_0 {parent=base_back_skirt, loc=<0 0 0>, rot=<0 0 0 1>}
 dyn_back_skirt_1 {parent=dyn_back_skirt_0, loc=<7.294 0 0>, rot=<0 0 0 1>}
 dyn_back_skirt_2 {parent=dyn_back_skirt_1, loc=<7.294 0 0>, rot=<0 0 0 1>}
 dyn_back_skirt_3 {parent=dyn_back_skirt_2, loc=<7.294 0 0>, rot=<0 0 0 1>}
 thigh_front_skirt_cons_offset {parent=pelvis, loc=<-1.0853 -0.2109 8.7053>, rot=<0.0355 -0.0009 0.9992 0.0155>}
 thigh_front_skirt_cons {parent=thigh_front_skirt_cons_offset, loc=<0 0 0>, rot=<0 0 0 1>}
 thigh_back_skirt_cons_offset {parent=pelvis, loc=<-1.0853 -0.2109 8.7053>, rot=<0.0355 -0.0009 0.9992 0.0155>}
 thigh_back_skirt_cons {parent=thigh_back_skirt_cons_offset, loc=<0 0 0>, rot=<0 0 0 1>}
 thigh_front_skirt_modify_offset {parent=pelvis, loc=<-1.0853 -0.2109 8.7053>, rot=<0.0355 -0.0009 0.9992 0.0155>}
 thigh_front_skirt_modify {parent=thigh_front_skirt_modify_offset, loc=<0 0 0>, rot=<0 0 0 1>}
 thigh_back_skirt_modify_offset {parent=pelvis, loc=<-1.0853 -0.2109 8.7053>, rot=<0.0355 -0.0009 0.9992 0.0155>}
 thigh_back_skirt_modify {parent=thigh_back_skirt_modify_offset, loc=<0 0 0>, rot=<0 0 0 1>}
 ik_foot_root {parent=root, loc=<0 0 0>, rot=<0 0 0 1>}
 ik_foot_l {parent=ik_foot_root, loc=<14.8262 -7.3545 8.296>, rot=<0.0357 0.7045 0.0284 -0.7082>}
 ik_foot_r {parent=ik_foot_root, loc=<-14.8262 -7.3546 8.2958>, rot=<0.7082 0.0284 -0.7045 0.0357>}
 ik_hand_root {parent=root, loc=<0 0 0>, rot=<0 0 0 1>}
 ik_hand_gun {parent=ik_hand_root, loc=<-48.1744 3.3546 104.4381>, rot=<0.6276 0.1309 0.4183 -0.6434>}
 ik_hand_l {parent=ik_hand_gun, loc=<59.3272 67.6997 34.3524>, rot=<-0.6158 -0.7027 -0.3565 0>}
 ik_hand_r {parent=ik_hand_gun, loc=<0 0 0>, rot=<0 0 0 1>}
 thigh_l_ctrl {controlType=EULER_TRANSFORM, type=Control, loc=<0 0 0>, rot=<0 0 0 1>}
 thigh_r_ctrl {controlType=EULER_TRANSFORM, type=Control, loc=<0 0 0>, rot=<0 0 0 1>}
 upperarm_l_ctrl {controlType=EULER_TRANSFORM, type=Control, loc=<0 0 0>, rot=<0 0 0 1>}
]
variables [
 IsSliding {type=bool, default=False}
 IsSlidingBaseball {type=bool, default=False}
 isTargeting {type=double, default=0}
 lowToGround {type=double, default=0}
]
