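# Physics asset for the jacket of the female medium SpyHandler player cosmetic, defining collision and joints for its dangling panels.
# Capsules on dyn_jacket bones hang from tiny base_jacket anchor capsules, with two-segment chains on the 0 left/right pair and single segments elsewhere, including one back panel.
# Each constraint has a nonzero Swing1 limit with Swing2 and twist limits at zero.
# Larger pelvis and thigh capsules, unconstrained here, provide body collision shapes.
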PhysicsAsset /BRCosmetics/Characters/Player/Female/Medium/Bodies/F_MED_SpyHandler/Meshes/F_MED_SpyHandler_Jacket_Physics
bodies {
 base_jacket_0_l [{shapeType=Capsule, length=0.505, radius=0.505, shapeName=base_jacket_0_l_capsule, rotation=<rot 0 -90 0>}]
 base_jacket_0_r [{shapeType=Capsule, length=0.505, radius=0.505, shapeName=base_jacket_0_r_capsule, rotation=<rot 0 -90 0>}]
 base_jacket_1_l [{shapeType=Capsule, length=0.505, radius=0.505, shapeName=base_jacket_1_l_capsule, rotation=<rot 0 -90 0>}]
 base_jacket_1_r [{shapeType=Capsule, length=0.505, radius=0.505, shapeName=base_jacket_1_r_capsule, rotation=<rot 0 -90 0>}]
 base_jacket_2_l [{shapeType=Capsule, length=0.505, radius=0.505, shapeName=base_jacket_2_l_capsule, rotation=<rot 0 -90 0>}]
 base_jacket_2_r [{shapeType=Capsule, length=0.505, radius=0.505, shapeName=base_jacket_2_r_capsule, rotation=<rot 0 -90 0>}]
 base_jacket_back_l [{shapeType=Capsule, length=0.505, radius=0.505, shapeName=base_jacket_back_l_capsule, rotation=<rot 0 -90 0>}]
 dyn_jacket_0_end_l [{shapeType=Capsule, length=6.9899, radius=1.3228, shapeName=dyn_jacket_0_end_l_capsule, center=<4.9936 0.9603 1.8587>, rotation=<rot 90 0 0>}]
 dyn_jacket_0_end_r [{
   shapeType=Capsule
   length=7.1158
   radius=1.3228
   shapeName=dyn_jacket_0_end_l_capsule
   center=<2 1.1873 -1.1966>
   rotation=<rot -62.7159 -93.0288 91.0972>
  }]
 dyn_jacket_0_l [{shapeType=Capsule, length=5.9559, radius=1.4531, shapeName=dyn_jacket_0_l_capsule, center=<9.1387 0.7408 1.7351>, rotation=<rot 90 0 0>}]
 dyn_jacket_0_r [{
   shapeType=Capsule
   length=6.0817
   radius=1.4531
   shapeName=dyn_jacket_0_l_capsule
   center=<2 1.4076 -1.2053>
   rotation=<rot 68.942 81.3049 85.661>
  }]
 dyn_jacket_1_l [{shapeType=Capsule, length=9.2464, radius=1.5, shapeName=dyn_jacket_1_l_capsule, center=<7.5 0 0>, rotation=<rot 90 0 0>}]
 dyn_jacket_1_r [{shapeType=Capsule, length=9.2464, radius=1.5, shapeName=dyn_jacket_1_r_capsule, center=<7.5 0 0>, rotation=<rot 90 0 0>}]
 dyn_jacket_2_l [{shapeType=Capsule, length=9.2464, radius=1.5, shapeName=dyn_jacket_2_l_capsule, center=<7.5 0 0>, rotation=<rot 90 0 0>}]
 dyn_jacket_2_r [{shapeType=Capsule, length=9.2464, radius=1.5, shapeName=dyn_jacket_2_r_capsule, center=<7.5 0 0>, rotation=<rot 90 0 0>}]
 dyn_jacket_back_l [{shapeType=Capsule, length=9.2464, radius=1.5, shapeName=dyn_jacket_back_l_capsule, center=<7.5 0 0>, rotation=<rot 90 0 0>}]
 pelvis [{shapeType=Capsule, length=7.8627, radius=9.7159, shapeName=pelvis_capsule, center=<2.6768 -1.7831 0>}]
 thigh_l [{
   shapeType=Capsule
   length=21.012
   radius=9.3861
   shapeName=thigh_l_capsule
   center=<-12.3401 -0.9019 0.4594>
   rotation=<rot 81.5106 176.8789 -179.8021>
  }]
 thigh_r [{
   shapeType=Capsule
   length=18.3056
   radius=8.8769
   shapeName=thigh_l_capsule
   center=<12.3401 0.9019 0.4594>
   rotation=<rot 81.5106 176.8789 -179.8021>
  }]
}
constraints [
 {bone1Name=dyn_jacket_0_l, bone2Name=base_jacket_0_l, swing1LimitDegrees=5, swing2LimitDegrees=0, twistLimitDegrees=0}
 {bone1Name=dyn_jacket_0_end_l, bone2Name=dyn_jacket_0_l, swing1LimitDegrees=15, swing2LimitDegrees=0, twistLimitDegrees=0}
 {bone1Name=dyn_jacket_0_r, bone2Name=base_jacket_0_r, swing1LimitDegrees=10, swing2LimitDegrees=0, twistLimitDegrees=0}
 {bone1Name=dyn_jacket_0_end_r, bone2Name=dyn_jacket_0_r, swing1LimitDegrees=60, swing2LimitDegrees=0, twistLimitDegrees=0}
 {bone1Name=dyn_jacket_1_l, bone2Name=base_jacket_1_l, swing1LimitDegrees=3, swing2LimitDegrees=0, twistLimitDegrees=0}
 {bone1Name=dyn_jacket_1_r, bone2Name=base_jacket_1_r, swing1LimitDegrees=10, swing2LimitDegrees=0, twistLimitDegrees=0}
 {bone1Name=dyn_jacket_2_l, bone2Name=base_jacket_2_l, swing1LimitDegrees=4, swing2LimitDegrees=0, twistLimitDegrees=0}
 {bone1Name=dyn_jacket_2_r, bone2Name=base_jacket_2_r, swing1LimitDegrees=4, swing2LimitDegrees=0, twistLimitDegrees=0}
 {bone1Name=dyn_jacket_back_l, bone2Name=base_jacket_back_l, swing1LimitDegrees=3, swing2LimitDegrees=0, twistLimitDegrees=0}
]
