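# Physics asset for the SpyHandler female medium body's skirt, defining collision and joints for two dynamic bone chains, skirt and back skirt.
# Each four-bone capsule chain hangs from a tiny base body, with joints limited to small single-axis swings and zero twist, and an extra crosswise capsule on its third bone.
# Pelvis, both thighs and the right calf add leg-area capsules, with only the right knee constrained, as fully free rotation.
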PhysicsAsset /BRCosmetics/Characters/Player/Female/Medium/Bodies/F_MED_SpyHandler/Meshes/F_MED_SpyHandler_Skirt_Physics
bodies {
 base_back_skirt [{shapeType=Capsule, length=0.505, radius=0.505, shapeName=base_back_skirt_capsule, rotation=<rot 0 -90 0>}]
 base_skirt_0 [{shapeType=Capsule, length=1.454, radius=1.5991, shapeName=base_skirt_0_capsule, rotation=<rot 0 -90 0>}]
 calf_r [{shapeType=Capsule, length=54.5812, radius=7.049, shapeName=calf_r_capsule, center=<22.35 2.7926 0>, rotation=<rot 90 180 176.8316>}]
 dyn_back_skirt_0 [{
   shapeType=Capsule
   length=13.0066
   radius=4.6391
   shapeName=dyn_back_skirt_0_capsule
   center=<2.0552 -0.102 5.2272>
   rotation=<rot 77.1878 -82.1357 5.8826>
  }]
 dyn_back_skirt_1 [{
   shapeType=Capsule
   length=13.0066
   radius=4.6391
   shapeName=dyn_back_skirt_1_capsule
   center=<2.0805 0.0917 3.8856>
   rotation=<rot 77.2699 -78.4234 9.5031>
  }]
 dyn_back_skirt_2 [
  {
   shapeType=Capsule
   length=13.0066
   radius=4.6391
   shapeName=dyn_back_skirt_2_capsule
   center=<2.2167 -0.0143 3.4715>
   rotation=<rot 76.8138 -58.548 28.8823>
  }
  {
   shapeType=Capsule
   length=12.0888
   radius=2
   shapeName=dyn_back_skirt_0_capsule
   center=<6.4346 -7.8674 -5.5991>
   rotation=<rot 88.9939 152.5028 152.861>
  }
 ]
 dyn_back_skirt_3 [{
   shapeType=Capsule
   length=13.0066
   radius=4.6391
   shapeName=dyn_back_skirt_3_capsule
   center=<2.1854 -0.0857 2.9583>
   rotation=<rot 77.0512 -62.9812 24.5638>
  }]
 dyn_skirt_0 [{
   shapeType=Capsule
   length=11.5495
   radius=1.9505
   shapeName=dyn_skirt_0_capsule
   center=<3.2291 -2.2945 -1.3282>
   rotation=<rot -11.3634 -1.4162 7.1504>
  }]
 dyn_skirt_1 [{
   shapeType=Capsule
   length=11.5495
   radius=1.9505
   shapeName=dyn_skirt_1_capsule
   center=<2.9933 -1.7698 -2.3824>
   rotation=<rot -11.3634 -1.4162 7.1505>
  }]
 dyn_skirt_2 [
  {
   shapeType=Capsule
   length=11.5495
   radius=1.9505
   shapeName=dyn_skirt_2_capsule
   center=<2.7761 -1.5853 -3.3589>
   rotation=<rot -11.3634 -1.4162 7.1505>
  }
  {
   shapeType=Capsule
   length=10.9666
   radius=2.2334
   shapeName=dyn_skirt_2_capsule
   center=<2.9959 1.9263 3.3482>
   rotation=<rot -86.1015 179.9658 -179.1526>
  }
 ]
 dyn_skirt_3 [{
   shapeType=Capsule
   length=11.5495
   radius=1.9505
   shapeName=dyn_skirt_3_capsule
   center=<3.2999 -1.3875 -4.045>
   rotation=<rot -11.3634 -1.4162 7.1505>
  }]
 pelvis [{shapeType=Capsule, length=12.6807, radius=8.8351, shapeName=pelvis_capsule, center=<-2.9822 -2.1495 0>}]
 thigh_l [{
   shapeType=Capsule
   length=28.9968
   radius=7.4378
   shapeName=thigh_r_capsule
   center=<-15.0245 0.5491 -0.2947>
   rotation=<rot 90 180 -175.4345>
  }]
 thigh_r [{
   shapeType=Capsule
   length=27.9427
   radius=8.4919
   shapeName=thigh_r_capsule
   center=<15.0245 0.5491 -0.2947>
   rotation=<rot 90 180 -175.4345>
  }]
}
constraints [
 {bone1Name=dyn_skirt_0, bone2Name=base_skirt_0, swing1LimitDegrees=10, swing2LimitDegrees=0, twistLimitDegrees=0}
 {bone1Name=dyn_skirt_2, bone2Name=dyn_skirt_1, swing1LimitDegrees=15, swing2LimitDegrees=0, twistLimitDegrees=0}
 {bone1Name=dyn_skirt_3, bone2Name=dyn_skirt_2, swing1LimitDegrees=10, swing2LimitDegrees=0, twistLimitDegrees=0}
 {bone1Name=dyn_skirt_1, bone2Name=dyn_skirt_0, swing1LimitDegrees=10, swing2LimitDegrees=0, twistLimitDegrees=0}
 {bone1Name=dyn_back_skirt_0, bone2Name=base_back_skirt, swing1LimitDegrees=0, swing2LimitDegrees=10, twistLimitDegrees=0}
 {bone1Name=dyn_back_skirt_1, bone2Name=dyn_back_skirt_0, swing1LimitDegrees=0, swing2LimitDegrees=20, twistLimitDegrees=0}
 {bone1Name=dyn_back_skirt_2, bone2Name=dyn_back_skirt_1, swing1LimitDegrees=0, swing2LimitDegrees=20, twistLimitDegrees=0}
 {bone1Name=dyn_back_skirt_3, bone2Name=dyn_back_skirt_2, swing1LimitDegrees=0, swing2LimitDegrees=20, twistLimitDegrees=0}
 {
  bone1Name=calf_r
  bone2Name=thigh_r
  swing1LimitDegrees=45
  swing1Motion=Free
  swing2LimitDegrees=45
  swing2Motion=Free
  twistLimitDegrees=45
  twistMotion=Free
 }
]
