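# Post-animation Blueprint for the female SpyHandler body, adding secondary motion to outfit and accessories.
# Copies the attached parent's pose, then chains cached passes: weighted gun/knife base constraints, Control Rig, bone offsets, then physics.
# RigidBody simulates jacket and skirt; AnimDynamics swings bracelets, knife and necklace.
# The event graph reduces skirt simulation while sprinting and derives strafe and targeting alphas for skirt offsets, which also react to skydiving and DBNO.
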AnimBlueprint /BRCosmetics/Characters/Player/Female/Medium/Bodies/F_MED_SpyHandler/Meshes/F_MED_SpyHandler_AnimBP
target_skeleton=&/F_MED_SpyHandler_Skeleton
graphs [
 AnimGraph {nodes [
   Output Pose(Result=Use cached pose 'dynamicsCache'()) {layerGroup=None}
   Save cached pose 'inputPoseCache'(Pose=Copy Pose From Mesh() {bUseAttachedParent=true}) {cacheName=inputPoseCache}
   n1 = Control Rig!(
    Source=Use cached pose 'surfaceConstraintsCache'()
    Alpha
    IsSliding:bool=Get bIsSliding()
    IsSlidingBaseball:bool=Get bIsBaseballSliding()
    isTargeting:double=Get isTargeting() -> (isTargeting:double)
   ) -> (Pose) {
    alphaBoolBlend {blendOption=Linear}
    controlRigAssetReference {blueprintRigClass=&/F_MED_SpyHandler_CtrlRig.F_MED_SpyHandler_CtrlRig_C}
   }
   Save cached pose 'controlRigCache'(Pose=@n1) {cacheName=controlRigCache}
   n2 = "Constraint - base_gun (PPPPPPPP)"!(
    ComponentPose=Local To Component(LocalPose=Use cached pose 'inputPoseCache'())
    Alpha=1
    ConstraintWeights_0=0.6562
    ConstraintWeights_1:float=0.0957
    ConstraintWeights_2:float=0.0683
    ConstraintWeights_3:float=0.0639
    ConstraintWeights_4:float=0.0598
    ConstraintWeights_5:float=0.0235
    ConstraintWeights_6:float=0.0217
    ConstraintWeights_7:float=0.0109
   ) -> (Pose) {
    alphaBoolBlend {blendOption=Linear}
    boneToModify {boneName=base_gun}
    constraintSetup [
     {targetBone {boneName=spine_04}}
     {targetBone {boneName=spine_02}}
     {targetBone {boneName=spine_03}}
     {targetBone {boneName=spine_05}}
     {targetBone {boneName=deform_pec_l}}
     {targetBone {boneName=clavicle_l}}
     {targetBone {boneName=upperarm_twist_01_l}}
     {targetBone {boneName=spine_01}}
    ]
    constraintWeights [1,1,1,1,1,1,1,1]
   }
   n3 = "Constraint - base_knife (PP)"!(ComponentPose=@n2, Alpha=1, ConstraintWeights_0=0.5365, ConstraintWeights_1:float=0.4635) -> (Pose) {
    alphaBoolBlend {blendOption=Linear}
    boneToModify {boneName=base_knife}
    constraintSetup [{targetBone {boneName=calf_twist_02_l}}, {targetBone {boneName=calf_l}}]
    constraintWeights [1,1]
   }
   Save cached pose 'surfaceConstraintsCache'(Pose=Component To Local(ComponentPose=@n3)) {cacheName=surfaceConstraintsCache}
   n4 = RigidBody!(ComponentPose=Local To Component(LocalPose=Use cached pose 'ModifyTransformCache'()), Alpha, OverrideWorldGravity:Vector="0,0,-980") -> (Pose) {
    bOverrideWorldGravity=true
    overridePhysicsAsset=&/F_MED_SpyHandler_Jacket_Physics
    simulationSpace=BaseBoneSpace
    alphaBoolBlend {blendOption=Linear}
    baseBoneRef {boneName=spine_02}
    componentAppliedLinearAccClamp=<3000 3000 3000>
    componentLinearAccScale=<0.55 0.5 0.5>
    componentLinearVelScale=<0.55 0.5 0.5>
    simSpaceSettings {velocityScaleZ=0, worldAlpha=0.2}
   }
   n5 = Anim Dynamics - Bone: dyn_frontBracelete_00_l(ComponentPose=@n4, Alpha=1, ExternalForce="0,0,0") {
    angularDampingOverride=1
    angularSpringConstant=10000
    bAngularSpring=true
    bOverrideAngularDamping=true
    bUseGravityOverride=true
    gravityScale=0.35
    simulationSpace=BoneRelative
    alphaBoolBlend {blendOption=Linear}
    boundBone {boneName=dyn_frontBracelete_00_l}
    componentAppliedLinearAccClamp=<3000 3000 3000>
    componentLinearAccScale=<0.5 0.5 0.5>
    componentLinearVelScale=<0.5 0.5 0.5>
    gravityOverride {z=-980}
    physicsBodyDefinitions [{
      sphereCollisionRadius=0
      boundBone {boneName=dyn_frontBracelete_00_l}
      boxExtents=<25 25 25>
      constraintSetup {angularLimitsMax=<0 0.2 -0.5>, angularLimitsMin=<0 -0.2 -1>, angularTarget=<1 0 0>}
      localJointOffset=<4 0 0>
     }]
    relativeSpaceBone {boneName=root}
   }
   n6 = Anim Dynamics - Bone: dyn_backBracelete_00_l(ComponentPose=@n5, Alpha=1, ExternalForce="0,0,0") {
    angularDampingOverride=1
    angularSpringConstant=15000
    bAngularSpring=true
    bOverrideAngularDamping=true
    bUseGravityOverride=true
    gravityScale=0.35
    simulationSpace=BoneRelative
    alphaBoolBlend {blendOption=Linear}
    boundBone {boneName=dyn_backBracelete_00_l}
    componentAppliedLinearAccClamp=<3000 3000 3000>
    componentLinearAccScale=<0.4 0.4 0.4>
    componentLinearVelScale=<0.4 0.4 0.4>
    gravityOverride {z=-980}
    physicsBodyDefinitions [{
      boundBone {boneName=dyn_backBracelete_00_l}
      boxExtents=<28 28 28>
      constraintSetup {angularLimitsMax=<0 0.5 -0.4>, angularLimitsMin=<0 -0.5 -0.5>, angularTarget=<1 0 0>}
      localJointOffset=<4 0 0>
     }]
    relativeSpaceBone {boneName=root}
   }
   n7 = Anim Dynamics - Bone: dyn_CP_frontBracelete_l(ComponentPose=@n6, Alpha=1, ExternalForce="0,0,0") {
    angularDampingOverride=1
    angularSpringConstant=14000
    bAngularSpring=true
    bOverrideAngularDamping=true
    bUseGravityOverride=true
    gravityScale=0.35
    simulationSpace=BoneRelative
    alphaBoolBlend {blendOption=Linear}
    boundBone {boneName=dyn_CP_frontBracelete_l}
    componentAppliedLinearAccClamp=<3000 3000 3000>
    componentLinearAccScale=<0.5 0.5 0.5>
    componentLinearVelScale=<0.5 0.5 0.5>
    gravityOverride {z=-980}
    physicsBodyDefinitions [{
      sphereCollisionRadius=0
      boundBone {boneName=dyn_CP_frontBracelete_l}
      boxExtents=<30 30 30>
      constraintSetup {angularLimitsMax=<0 0 3>, angularLimitsMin=<0 0 -22>, angularTarget=<1 0 0>}
      localJointOffset=<10 0 0>
     }]
    relativeSpaceBone {boneName=root}
   }
   n8 = Anim Dynamics - Bone: dyn_CP_backBracelete_l(ComponentPose=@n7, Alpha=1, ExternalForce="0,0,0") {
    angularDampingOverride=1
    angularSpringConstant=15000
    bAngularSpring=true
    bOverrideAngularDamping=true
    bUseGravityOverride=true
    gravityScale=0.35
    simulationSpace=BoneRelative
    alphaBoolBlend {blendOption=Linear}
    boundBone {boneName=dyn_CP_backBracelete_l}
    componentAppliedLinearAccClamp=<3000 3000 3000>
    componentLinearAccScale=<0.4 0.4 0.4>
    componentLinearVelScale=<0.4 0.4 0.4>
    gravityOverride {z=-980}
    physicsBodyDefinitions [{
      boundBone {boneName=dyn_CP_backBracelete_l}
      boxExtents=<35 35 35>
      constraintSetup {angularLimitsMax=<0 0 10>, angularLimitsMin=<0 0 -20>, angularTarget=<1 0 0>}
      localJointOffset=<10 0 0>
     }]
    relativeSpaceBone {boneName=root}
   }
   n9 = Anim Dynamics - Bone: dyn_knife!(ComponentPose=@n8, Alpha=1, GravityOverride:Vector="0,0,-980") -> (Pose) {
    angularDampingOverride=1
    angularSpringConstant=2500
    bAngularSpring=true
    bGravityOverrideInSimSpace=true
    bOverrideAngularDamping=true
    bUseGravityOverride=true
    gravityScale=0.35
    linearSpringConstant=100
    simulationSpace=RootRelative
    alphaBoolBlend {blendOption=Linear}
    boundBone {boneName=dyn_knife}
    componentAppliedLinearAccClamp=<3000 3000 3000>
    componentLinearAccScale=<0.05 0.05 0.05>
    componentLinearVelScale=<0.05 0.05 0.05>
    gravityOverride {z=-980}
    physicsBodyDefinitions [{
      boundBone {boneName=dyn_knife}
      boxExtents=<25 25 15>
      constraintSetup {angularLimitsMax=<0 5 5>, angularLimitsMin=<0 -5 -5>, angularTarget=<1 0 0>}
      localJointOffset=<6 0 0>
     }]
    relativeSpaceBone {boneName=calf_l}
   }
   n10 = RigidBody(ComponentPose=@n9, Alpha=Get isSprinting() -> (isSprinting:double), OverrideWorldGravity="0,0,0") {
    bClampLinearTranslationLimitToRefPose=true
    bOverrideWorldGravity=true
    overridePhysicsAsset=&/F_MED_SpyHandler_Skirt_Physics
    simulationSpace=BaseBoneSpace
    alphaBoolBlend {blendOption=Linear}
    alphaScaleBiasClamp {bInterpResult=true}
    baseBoneRef {boneName=pelvis}
    componentAppliedLinearAccClamp=<3000 3000 3000>
    componentLinearAccScale=<0.3 0.3 0.3>
    componentLinearVelScale=<0.3 0.3 0.3>
    simSpaceSettings {velocityScaleZ=0}
   }
   n11 = Anim Dynamics - Bone: dyn_necklace(ComponentPose=@n10, Alpha=1, ExternalForce="0,0,0") {
    angularDampingOverride=1
    angularSpringConstant=20000
    bAngularSpring=true
    bOverrideAngularDamping=true
    bUseGravityOverride=true
    gravityScale=0.35
    simulationSpace=RootRelative
    alphaBoolBlend {blendOption=Linear}
    boundBone {boneName=dyn_necklace}
    componentAppliedLinearAccClamp=<3000 3000 3000>
    componentLinearAccScale=<1 1 1>
    componentLinearVelScale=<1 1 1>
    gravityOverride {z=-980}
    physicsBodyDefinitions [{
      boundBone {boneName=dyn_necklace}
      boxExtents=<40 40 40>
      constraintSetup {angularLimitsMax=<0 15 0>, angularLimitsMin=<0 -15 -45>, angularTarget=<1 0 0>}
      localJointOffset=<4 0 0>
     }]
    relativeSpaceBone {boneName=spine_05}
   }
   Save cached pose 'dynamicsCache'(Pose=Component To Local(ComponentPose=@n11)) {cacheName=dynamicsCache}
   n12 = "Transform (Modify) Bone - Bone: base_back_skirt"!(
    ComponentPose=Local To Component(LocalPose=Use cached pose 'controlRigCache'())
    bAlphaBoolEnabled:bool=Get bIsSkyDiving()
    Translation="0,0,5"
    Rotation="0,0,0"
    Scale="1,1,1"
   ) -> (Pose) {
    alphaInputType=Bool
    rotationMode=BMM_Additive
    rotationSpace=BCS_BoneSpace
    translationMode=BMM_Additive
    translationSpace=BCS_BoneSpace
    alphaBoolBlend {blendInTime=1, blendOption=Linear, blendOutTime=1}
    alphaScaleBiasClamp {bInterpResult=true}
    boneToModify {boneName=base_back_skirt}
    translation {z=5}
   }
   n13 = "Transform (Modify) Bone - Bone: base_back_skirt"(ComponentPose=@n12, bAlphaBoolEnabled=Get bIsDBNO(), Translation="0,0,0", Rotation="-60,0,0", Scale="1,1,1") {
    alphaInputType=Bool
    rotationMode=BMM_Additive
    rotationSpace=BCS_BoneSpace
    translationMode=BMM_Additive
    translationSpace=BCS_BoneSpace
    alphaBoolBlend {blendInTime=1, blendOption=Linear, blendOutTime=1}
    alphaScaleBiasClamp {bInterpResult=true}
    boneToModify {boneName=base_back_skirt}
    rotation {pitch=-60}
   }
   n14 = "Transform (Modify) Bone - Bone: base_skirt_0"(ComponentPose=@n13, Alpha=Get Left() -> (Left:double), Translation="0,-1.5,0", Rotation="0,-15,0", Scale="1,1,1") {
    rotationMode=BMM_Additive
    rotationSpace=BCS_BoneSpace
    translationMode=BMM_Additive
    translationSpace=BCS_BoneSpace
    alphaBoolBlend {blendInTime=1, blendOption=Linear, blendOutTime=1}
    alphaScaleBiasClamp {bInterpResult=true}
    boneToModify {boneName=base_skirt_0}
    rotation {yaw=-15}
    translation {y=-1.5}
   }
   n15 = "Transform (Modify) Bone - Bone: base_skirt_0"(ComponentPose=@n14, Alpha=Get Right() -> (Right:double), Translation="0,-1.5,0", Rotation="0,-15,0", Scale="1,1,1") {
    rotationMode=BMM_Additive
    rotationSpace=BCS_BoneSpace
    translationMode=BMM_Additive
    translationSpace=BCS_BoneSpace
    alphaBoolBlend {blendInTime=1, blendOption=Linear, blendOutTime=1}
    alphaScaleBiasClamp {bInterpResult=true}
    boneToModify {boneName=base_skirt_0}
    rotation {yaw=-15}
    translation {y=-1.5}
   }
   n16 = "Transform (Modify) Bone - Bone: base_skirt_back"(ComponentPose=@n15, Alpha=Get Left Targeting() -> (Left Targeting:double), Translation="0,0,30", Rotation="-50,0,0", Scale="1,1,1") {
    rotationMode=BMM_Additive
    rotationSpace=BCS_BoneSpace
    translationMode=BMM_Additive
    translationSpace=BCS_BoneSpace
    alphaBoolBlend {blendInTime=1, blendOption=Linear, blendOutTime=1}
    alphaScaleBiasClamp {bInterpResult=true}
    boneToModify {boneName=base_skirt_back}
    rotation {pitch=-50}
    translation {z=30}
   }
   n17 = "Transform (Modify) Bone - Bone: base_skirt_back"(ComponentPose=@n16, Alpha=Get Right Targeting() -> (Right Targeting:double), Translation="0,0,30", Rotation="0,0,0", Scale="1,1,1") {
    rotationMode=BMM_Additive
    rotationSpace=BCS_BoneSpace
    translationMode=BMM_Additive
    translationSpace=BCS_BoneSpace
    alphaBoolBlend {blendInTime=1, blendOption=Linear, blendOutTime=1}
    alphaScaleBiasClamp {bInterpResult=true}
    boneToModify {boneName=base_skirt_back}
    translation {z=30}
   }
   n18 = "Transform (Modify) Bone - Bone: base_back_skirt"!(ComponentPose=@n17, bAlphaBoolEnabled=True, Translation="0,-2,0", Rotation="10,0,0") -> (Pose) {
    alphaInputType=Bool
    rotationMode=BMM_Additive
    rotationSpace=BCS_BoneSpace
    translationMode=BMM_Additive
    translationSpace=BCS_BoneSpace
    alphaBoolBlend {blendInTime=1, blendOption=Linear, blendOutTime=1}
    boneToModify {boneName=base_back_skirt}
    rotation {pitch=10}
    translation {y=-2}
   }
   n19 = "Transform (Modify) Bone - Bone: leg_fix_asw_l"!(ComponentPose=@n18, bAlphaBoolEnabled:bool=True, Scale="0.8,0.8,0.8") -> (Pose) {
    alphaInputType=Bool
    rotationSpace=BCS_BoneSpace
    scaleMode=BMM_Additive
    scaleSpace=BCS_BoneSpace
    translationSpace=BCS_BoneSpace
    alphaBoolBlend {blendInTime=1, blendOption=Linear, blendOutTime=1}
    boneToModify {boneName=leg_fix_asw_l}
    scale=<0.8 0.8 0.8>
   }
   n20 = "Transform (Modify) Bone - Bone: leg_fix_asw_r"!(ComponentPose=@n19, bAlphaBoolEnabled:bool=True, Scale="0.8,0.8,0.8") -> (Pose) {
    alphaInputType=Bool
    rotationSpace=BCS_BoneSpace
    scaleMode=BMM_Additive
    scaleSpace=BCS_BoneSpace
    translationSpace=BCS_BoneSpace
    alphaBoolBlend {blendInTime=1, blendOption=Linear, blendOutTime=1}
    boneToModify {boneName=leg_fix_asw_r}
    scale=<0.8 0.8 0.8>
   }
   Save cached pose 'ModifyTransformCache'(Pose=Component To Local(ComponentPose=@n20)) {cacheName=ModifyTransformCache}
  ]}
 EventGraph {nodes [
   Event BlueprintUpdateAnimation()
   > n1 = Sequence!(execute) -> (then_0, then_1, then_2:Exec, then_3:Exec)
   > n2 = Gravity Override State Function DEPRECATED(CharacterPartAnimInstance=Self-Reference(), GameTimeInSeconds=GetGameTimeInSeconds(), GravityOverrideParams=Get Gravity Override Params())
   > Set GravityOverride(GravityOverride=@n2)
   > n3 = Branch(Condition=OR Boolean(A=Get bIsSprinting(), B=Get bIsCrouchSprinting()))
   > Set isSprinting(isSprinting:double=0.2) -> (Output_Get:double)
   Set isSprinting(execute=@n3.else, isSprinting=1)
   n4 = PlayerMovement(execute=@n1.then_1, AnimInstance=Self-Reference(), XDotProductCoefficient=100, YDotProductCoefficient=100, ZDotProductCoefficient=100)
   n5 = BreakVector(InVec=MakeVector(X=@n4, Y=@n4.YMovement))
   > n6 = Set Left(Left:double="To Float (Boolean)"(InBool="float < float"(A=@n5, B=-50))) -> (Output_Get:double)
   n7 = Set isTargeting(execute=@n10.then, isTargeting:double=1) -> (Output_Get:double)
   Set Left Targeting(execute=@n6.then, Left Targeting:double=float * float(A=@n6, B=@n7) {is=title K2Node_PromotableOperator}) -> (Output_Get:double)
   > n8 = Set Right(Right:double="To Float (Boolean)"(InBool="float > float"(A=@n5, B=50))) -> (Output_Get:double)
   > Set Right Targeting(Right Targeting:double=float * float(A=@n8, B=@n7)) -> (Output_Get:double)
   n9 = Get bIsCrouching()
   Set IsMovingStanding(execute=@n1.then_2, IsMovingStanding:bool=AND Boolean(A=NOT Boolean(A=@n9), B=Get bIsMoving2D())) -> (Output_Get:bool)
   Set NewVar(execute=@n1.then_3, NewVar:bool=NOT Boolean(A=Get bRigidBodyEnabled())) -> (Output_Get:bool)
   n10 = Event OnBeginTargeting() {is=title K2Node_Event}
   Event OnEndTargeting() {is=title K2Node_Event}
   > Set isTargeting(isTargeting=0)
  ]}
]
variables [
 Gravity Override Params {type=GravityOverrideParamsStruct}
 GravityOverride {type=Vector}
 IsMovingStanding {type=boolean}
 isSprinting {type=number}
 isTargeting {type=number}
 Left {type=number}
 Left Targeting {type=number}
 NewVar {type=boolean}
 Right {type=number}
 Right Targeting {type=number}
]
