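# Physics asset for the SpyHandler female face-accessory mesh part (BackBling variant by name), covering hair dynamics and upper-body collision shapes.
# Capsules cover head, neck, spine, clavicles and upper arms, linked by free constraints, with small base bodies linking hair to the head.
# Back, side and front hair chains mostly use tight swing limits, some with locked twist.
# Right front pieces hang off the right back-hair root, and FrontHair_L lacks constraints between its middle segments.
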PhysicsAsset /BRCosmetics/Characters/Player/Female/Medium/Bodies/F_MED_SpyHandler/Meshes/Parts/F_MED_SpyHandler_FaceAcc_BackBling_Physics
bodies {
 base_BackHair_L [{shapeType=Capsule, length=0, radius=0.5, shapeName=base_BackHair_L_capsule, rotation=<rot 90 0 0>}]
 base_BackHair_M [{shapeType=Capsule, length=0, radius=0.5, shapeName=base_BackHair_M_capsule, rotation=<rot 90 0 0>}]
 base_BackHair_R [{shapeType=Capsule, length=0, radius=0.5, shapeName=base_BackHair_R_capsule, rotation=<rot 90 0 0>}]
 base_FrontHair_L [{shapeType=Capsule, length=0, radius=0.5, shapeName=base_FrontHair_L_capsule, rotation=<rot 90 0 0>}]
 base_SideHairA_L [{shapeType=Capsule, length=0, radius=0.5, shapeName=base_SideHairA_L_capsule, rotation=<rot 90 0 0>}]
 clavicle_l [{
   shapeType=Capsule
   length=9
   radius=6.25
   shapeName=clavicle_r_capsule
   center=<4.8507 -0.65 -1.3777>
   rotation=<rot 76.4723 -0.0009 179.999>
  }]
 clavicle_r [{
   shapeType=Capsule
   length=7.5004
   radius=6.5476
   shapeName=clavicle_r_capsule
   center=<-5.433 1.3318 0.7404>
   rotation=<rot 75.8544 -0.002 -0.0021>
  }]
 dyn_BackHairPiece_R_01 [{
   shapeType=Capsule
   length=4.1747
   radius=2.15
   shapeName=dyn_BackHair_R_03_capsule
   center=<1.5253 0 -0.5524>
   rotation=<rot 80.0003 179.9999 179.9999>
  }]
 dyn_BackHair_L_01 [{shapeType=Capsule, length=2, radius=0.77, shapeName=dyn_BackHair_L_01_capsule, center=<1.5 0 0.5>, rotation=<rot 90 0 0>}]
 dyn_BackHair_L_02 [{shapeType=Capsule, length=2, radius=1.5, shapeName=dyn_BackHair_L_02_capsule, center=<2 0 0>, rotation=<rot 90 0 0>}]
 dyn_BackHair_L_03 [{shapeType=Capsule, length=3.6858, radius=2, shapeName=dyn_BackHair_L_03_capsule, center=<1.6174 0 -1.4612>, rotation=<rot 60.0003 180 -180>}]
 dyn_BackHair_L_04 [{
   shapeType=Capsule
   length=1.5
   radius=1.5
   shapeName=dyn_BackHair_L_04_capsule
   center=<1.5146 0.2079 1.6956>
   rotation=<rot 37.737 180 -137.8983>
  }]
 dyn_BackHair_M_01 [{shapeType=Capsule, length=2, radius=1, shapeName=dyn_BackHair_M_01_capsule, center=<1.5 0 0.5>, rotation=<rot 90 0 0>}]
 dyn_BackHair_M_02 [{shapeType=Capsule, length=2, radius=1.5, shapeName=dyn_BackHair_M_02_capsule, center=<1.5 -0.2586 0.2491>, rotation=<rot 90 0 0>}]
 dyn_BackHair_M_03 [{
   shapeType=Capsule
   length=5.0766
   radius=2.5761
   shapeName=dyn_BackHair_M_03_capsule
   center=<3.6541 0 -0.4695>
   rotation=<rot 83.1116 180 180>
  }]
 dyn_BackHair_M_05 [{
   shapeType=Capsule
   length=4.3524
   radius=2.25
   shapeName=dyn_BackHair_M_05_capsule
   center=<4.6593 -1.1839 0.7622>
   rotation=<rot 82.9594 -6.8427 0.0852>
  }]
 dyn_BackHair_R_01 [{
   shapeType=Capsule
   length=6.0495
   radius=1.25
   shapeName=dyn_BackHair_R_02_capsule
   center=<6.033 -1.3662 -0.4667>
   rotation=<rot 68.1872 157.4424 173.9112>
  }]
 dyn_BackHair_R_03 [{shapeType=Capsule, length=1.3106, radius=1.9394, shapeName=dyn_BackHair_R_03_capsule, center=<1.5 0 -0.2635>, rotation=<rot 90 0 0>}]
 dyn_BackHair_R_04 [{
   shapeType=Capsule
   length=7.1348
   radius=1.9734
   shapeName=dyn_BackHair_R_04_capsule
   center=<3.2418 -0.5156 0.0563>
   rotation=<rot 70.2911 179.9998 177.5467>
  }]
 dyn_FrontHairA_R_01 [{
   shapeType=Capsule
   length=2.6692
   radius=0.7302
   shapeName=dyn_FrontHairA_R_01_capsule
   center=<1.9247 -0.3776 0.408>
   rotation=<rot 0 77.5728 -73.9372>
  }]
 dyn_FrontHairB_R_01 [{
   shapeType=Capsule
   length=3.1499
   radius=1.1524
   shapeName=dyn_FrontHairB_R_01_capsule
   center=<2.1371 -0.8454 0.045>
   rotation=<rot 0 61.0407 -85.3097>
  }]
 dyn_FrontHair_L_01 [{
   shapeType=Capsule
   length=6.3254
   radius=1.7383
   shapeName=dyn_FrontHair_L_01_capsule
   center=<4.0947 0 0.4014>
   rotation=<rot 84.9857 -0.0003 -0.0004>
  }]
 dyn_FrontHair_L_02 [{
   shapeType=Capsule
   length=2
   radius=1.5
   shapeName=dyn_FrontHair_L_02_capsule
   center=<2.4715 0 0.2005>
   rotation=<rot 86.3066 -0.0003 -0.0003>
  }]
 dyn_FrontHair_L_03 [{shapeType=Capsule, length=2, radius=1.5, shapeName=dyn_FrontHair_L_03_capsule, center=<2 0 0>, rotation=<rot 90 0 0>}]
 dyn_FrontHair_L_04 [{
   shapeType=Capsule
   length=10.2305
   radius=1.75
   shapeName=dyn_FrontHair_L_04_capsule
   center=<1.7942 -0.0144 0.0958>
   rotation=<rot 81.4763 -167.4807 -162.9562>
  }]
 dyn_SideHairA_L_01 [{shapeType=Capsule, length=2, radius=1.5, shapeName=dyn_SideHairA_L_01_capsule, center=<2 0 0>, rotation=<rot 90 0 0>}]
 dyn_SideHairA_L_02 [{shapeType=Capsule, length=2, radius=1.5, shapeName=dyn_SideHairA_L_02_capsule, center=<2 0 0>, rotation=<rot 90 0 0>}]
 dyn_SideHairA_L_03 [{
   shapeType=Capsule
   length=1.889
   radius=2.111
   shapeName=dyn_SideHairA_L_03_capsule
   center=<3.2248 -1.1237 0.0567>
   rotation=<rot 90 -153.4349 -132.6269>
  }]
 head [{shapeType=Capsule, length=6.3438, radius=6.6, shapeName=head_capsule, center=<1.708 2.0602 0.5759>, rotation=<rot 0 -86.6615 90>}]
 neck_02 [{shapeType=Capsule, length=18.1708, radius=14.722, shapeName=neck_02_capsule, center=<1.3826 15.0112 0>}]
 spine_03 [{shapeType=Capsule, length=7.5049, radius=11.4951, shapeName=spine_03_capsule, center=<8.667 1.6116 0>, rotation=<rot 90 -179.9813 -179.799>}]
 spine_05 [
  {shapeType=Capsule, length=16.4431, radius=8.783, shapeName=spine_05_capsule, center=<1.8216 -0.9737 0>, rotation=<rot 0 -90 0>}
  {shapeType=Capsule, length=30.5, radius=15, shapeName=spine_05_capsule, center=<25.3124 -20.7377 0>}
  {shapeType=Capsule, length=15, radius=15, shapeName=spine_05_capsule, center=<7.1119 -29.2679 0>}
 ]
 upperarm_l [{
   shapeType=Capsule
   length=21.4
   radius=6.9
   shapeName=upperarm_l_capsule
   center=<11.4069 0 -0.7875>
   rotation=<rot 84.7743 -179.9942 -179.9946>
  }]
 upperarm_r [{
   shapeType=Capsule
   length=21.4
   radius=6.9
   shapeName=upperarm_l_capsule
   center=<-13.0462 0.5546 1.1252>
   rotation=<rot 84.7743 -179.9942 0.0054>
  }]
}
constraints [
 {
  bone1Name=dyn_BackHair_L_02
  bone2Name=dyn_BackHair_L_01
  swing1LimitDegrees=10
  swing2LimitDegrees=10
  twistLimitDegrees=0
  twistMotion=Locked
 }
 {
  bone1Name=dyn_BackHair_L_03
  bone2Name=dyn_BackHair_L_02
  swing1LimitDegrees=10
  swing2LimitDegrees=10
  twistLimitDegrees=0
  twistMotion=Locked
 }
 {bone1Name=dyn_BackHair_L_04, bone2Name=dyn_BackHair_L_03, swing1LimitDegrees=10, swing2LimitDegrees=10, twistLimitDegrees=0}
 {
  bone1Name=dyn_BackHair_M_02
  bone2Name=dyn_BackHair_M_01
  swing1LimitDegrees=2
  swing2LimitDegrees=10
  twistLimitDegrees=0
  twistMotion=Locked
 }
 {
  bone1Name=dyn_BackHair_M_03
  bone2Name=dyn_BackHair_M_02
  swing1LimitDegrees=2
  swing2LimitDegrees=10
  twistLimitDegrees=0
  twistMotion=Locked
 }
 {
  bone1Name=dyn_BackHair_M_05
  bone2Name=dyn_BackHair_M_03
  swing1LimitDegrees=10
  swing2LimitDegrees=10
  twistLimitDegrees=0
  twistMotion=Locked
 }
 {
  bone1Name=dyn_BackHair_R_03
  bone2Name=dyn_BackHair_R_01
  swing1LimitDegrees=5
  swing2LimitDegrees=10
  twistLimitDegrees=0
  twistMotion=Locked
 }
 {bone1Name=dyn_BackHair_R_04, bone2Name=dyn_BackHair_R_03, swing1LimitDegrees=10, swing2LimitDegrees=10, twistLimitDegrees=0}
 {bone1Name=dyn_FrontHair_L_04, bone2Name=dyn_FrontHair_L_03, swing1LimitDegrees=15, swing2LimitDegrees=20, twistLimitDegrees=0}
 {bone1Name=dyn_SideHairA_L_02, bone2Name=dyn_SideHairA_L_01, swing1LimitDegrees=15, swing2LimitDegrees=10, twistLimitDegrees=0}
 {bone1Name=dyn_SideHairA_L_03, bone2Name=dyn_SideHairA_L_02, swing1LimitDegrees=20, swing2LimitDegrees=20, twistLimitDegrees=0}
 {bone1Name=dyn_FrontHair_L_01, bone2Name=base_FrontHair_L, swing1LimitDegrees=7, swing2LimitDegrees=5, twistLimitDegrees=0}
 {bone1Name=dyn_SideHairA_L_01, bone2Name=base_SideHairA_L, swing1LimitDegrees=10, swing2LimitDegrees=8, twistLimitDegrees=0}
 {bone1Name=dyn_BackHair_L_01, bone2Name=base_BackHair_L, swing1LimitDegrees=0, swing2LimitDegrees=10, twistLimitDegrees=0}
 {bone1Name=dyn_BackHair_M_01, bone2Name=base_BackHair_M, swing1LimitDegrees=0, swing2LimitDegrees=10, twistLimitDegrees=0}
 {bone1Name=dyn_BackHair_R_01, bone2Name=base_BackHair_R, swing1LimitDegrees=0, swing2LimitDegrees=10, twistLimitDegrees=0}
 {
  bone1Name=head
  bone2Name=neck_02
  swing1LimitDegrees=45
  swing1Motion=Free
  swing2LimitDegrees=45
  swing2Motion=Free
  twistLimitDegrees=45
  twistMotion=Free
 }
 {
  bone1Name=neck_02
  bone2Name=spine_05
  swing1LimitDegrees=45
  swing1Motion=Free
  swing2LimitDegrees=45
  swing2Motion=Free
  twistLimitDegrees=45
  twistMotion=Free
 }
 {
  bone1Name=upperarm_l
  bone2Name=clavicle_l
  swing1LimitDegrees=45
  swing1Motion=Free
  swing2LimitDegrees=45
  swing2Motion=Free
  twistLimitDegrees=45
  twistMotion=Free
 }
 {
  bone1Name=base_BackHair_L
  bone2Name=head
  swing1LimitDegrees=45
  swing1Motion=Free
  swing2LimitDegrees=45
  swing2Motion=Free
  twistLimitDegrees=45
  twistMotion=Free
 }
 {
  bone1Name=base_BackHair_M
  bone2Name=head
  swing1LimitDegrees=45
  swing1Motion=Free
  swing2LimitDegrees=45
  swing2Motion=Free
  twistLimitDegrees=45
  twistMotion=Free
 }
 {
  bone1Name=base_BackHair_R
  bone2Name=head
  swing1LimitDegrees=45
  swing1Motion=Free
  swing2LimitDegrees=45
  swing2Motion=Free
  twistLimitDegrees=45
  twistMotion=Free
 }
 {
  bone1Name=base_FrontHair_L
  bone2Name=head
  swing1LimitDegrees=45
  swing1Motion=Free
  swing2LimitDegrees=45
  swing2Motion=Free
  twistLimitDegrees=45
  twistMotion=Free
 }
 {
  bone1Name=base_SideHairA_L
  bone2Name=head
  swing1LimitDegrees=45
  swing1Motion=Free
  swing2LimitDegrees=45
  swing2Motion=Free
  twistLimitDegrees=45
  twistMotion=Free
 }
 {
  bone1Name=clavicle_l
  bone2Name=spine_05
  swing1LimitDegrees=45
  swing1Motion=Free
  swing2LimitDegrees=45
  swing2Motion=Free
  twistLimitDegrees=45
  twistMotion=Free
 }
 {
  bone1Name=clavicle_r
  bone2Name=spine_05
  swing1LimitDegrees=45
  swing1Motion=Free
  swing2LimitDegrees=45
  swing2Motion=Free
  twistLimitDegrees=45
  twistMotion=Free
 }
 {bone1Name=dyn_BackHairPiece_R_01, bone2Name=dyn_BackHair_R_03, swing1LimitDegrees=15, swing2LimitDegrees=20, twistLimitDegrees=0}
 {
  bone1Name=spine_05
  bone2Name=spine_03
  swing1LimitDegrees=45
  swing1Motion=Free
  swing2LimitDegrees=45
  swing2Motion=Free
  twistLimitDegrees=45
  twistMotion=Free
 }
 {bone1Name=dyn_FrontHairB_R_01, bone2Name=dyn_BackHair_R_01, swing1LimitDegrees=15, swing2LimitDegrees=10, twistLimitDegrees=0}
 {bone1Name=dyn_FrontHairA_R_01, bone2Name=dyn_BackHair_R_01, swing1LimitDegrees=10, swing2LimitDegrees=0, twistLimitDegrees=0}
 {
  bone1Name=upperarm_r
  bone2Name=clavicle_r
  swing1LimitDegrees=45
  swing1Motion=Free
  swing2LimitDegrees=45
  swing2Motion=Free
  twistLimitDegrees=45
  twistMotion=Free
 }
 {
  bone1Name=dyn_BackHair_L_02
  bone2Name=dyn_SideHairA_L_02
  swing1LimitDegrees=45
  swing1Motion=Free
  swing2LimitDegrees=45
  swing2Motion=Free
  twistLimitDegrees=45
  twistMotion=Free
 }
]
